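# Translucent, unlit, two-sided slash trail material for SinisterMelon effects on Niagara ribbons and mesh particles.
# Opacity pans a noise-distorted mask, timed by dynamic MaskOffset, then applies UV edge, view-angle and depth fades.
# Emissive multiplies two curve-atlas gradients by particle color and scrolling detail noise, adds a tinted color layer, then optionally desaturates.
# Vertex wobble blends opposing time-driven sine waves along WPODirection, scaled by dynamic WPOIntensity.
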Material /SinisterMelon/Gameplay/SinisterMelon/FX/NewFX/Materials/M_SinisterMelon_BigSlashTrail
graph {
 properties {
  bAutomaticallySetUsageInEditor=false
  bUsedWithNiagaraMeshParticles=true
  bUsedWithNiagaraRibbons=true
  blendMode=BLEND_Translucent
  shadingModel=MSM_Unlit
  twoSided=true
  usedShadingModels=Unlit
 }
 nodes [
  n1 = ComponentMask(None=TextureCoordinate()) {r=true}
  n2 = OneMinus(None=@n1)
  n3 = Multiply(A=Sine(None=@n1) {period=2}, B=LinearInterpolate(B=Power(Base=@n2), Alpha=ScalarParameter() {parameterName=UseTailFade}) {constA=1})
  n4 = VectorParameter() {parameterName=UVScale, channelNames {a="", b="", g=V Scale, r=U Scale}, defaultValue=<rgba 0.3 1 0 0>}
  n5 = Multiply(A=TextureCoordinate(), B=ComponentMask(None=@n4) {g=true, r=true})
  n6 = VectorParameter() {parameterName=UVOffset, channelNames {a="", b="", g=V Offset, r=U Offset}, defaultValue=<rgba 0.35 0 0 0>}
  n7 = ComponentMask(None=@n6) {g=true, r=true}
  n8 = VectorParameter() {parameterName=DisturPanner, defaultValue=<rgba -1 0 0 0>}
  n9 = TextureSample(UVs=Panner(Coordinate=TextureCoordinate() {vTiling=0.7}, Speed=ComponentMask(None=@n8.RGB) {g=true, r=true}) {speedX=1}) {texture=/Game/Weapons/FORT_Melee/Pickaxe_CoatCheck/FX/T_Pickaxe_CoatCheck_Noise}
  n10 = ScalarParameter() {defaultValue=0.1, parameterName=Disturb}
  n11 = Panner(
   Coordinate=Add(A=Add(A=@n5, B=@n7), B=Multiply(A=@n9.R, B=@n10) {constB=0.2}) {constB=0}
   Time=DynamicParameter() {defaultValue=<rgba 0 0.2 0 10>, paramNames [Dissolve, DissolveOffset, MaskOffset, WPOIntensity]}.MaskOffset
   Speed=ComponentMask(None=VectorParameter() {parameterName=MaskPanner}.RGB) {g=true, r=true}
  )
  n12 = TextureCoordinate() {uTiling=0.5}
  n13 = Multiply(
   A=StaticComponentMaskParameter(None=TextureSampleParameter2D(UVs=@n11) {parameterName=Defalut Tex, texture=/Game/Accessories/FORT_Backpacks/Backpack_M_MED_OceanBreeze/FX/T_OceanBreeze_SplineA}.RGB) {defaultR=true, parameterName=Mask}
   B=LinearInterpolate(
    B=TextureSample(UVs=@n12) {samplerType=SAMPLERTYPE_Grayscale, texture=&LowResBlurredNoise}.R
    Alpha=ScalarParameter() {parameterName=UseDetailMask}
   ) {constA=1}
  )
  n14 = Multiply(
   A=LinearInterpolate(Alpha=VertexColor().R) {constA=2, constB=2}
   B=ScalarParameter() {defaultValue=1, parameterName="FadeSensitivity (optional) (S)"}
  )
  n15 = Multiply(A=Multiply(
    A=Multiply(A=@n13, B=@n3)
    B=MaterialFunctionCall("FadeSensitivity (optional)"=@n14) {materialFunction=&FadeBasedOnViewAngle}.Result
   ))
  n16 = Sine(None=ComponentMask(None=TextureCoordinate()) {g=true}) {period=2}
  DepthFade(
   Opacity=Multiply(A=Multiply(A=@n15, B=@n16), B=ParticleColor().A)
   FadeDistance=ScalarParameter() {defaultValue=50, parameterName=FadeDistance}
  ) {fadeDistanceDefault=50}
  n17 = TextureCoordinate() {vTiling=0.5}
  n18 = Constant2Vector() {r=0.7}
  n19 = VectorParameter() {parameterName=ColorTexPanner}
  n20 = TextureSample(UVs=Panner(Coordinate=Add(A=@n17, B=@n18), Speed=ComponentMask(None=@n19.RGB) {g=true, r=true})) {texture=&T_Tile_RainbowColor}
  n21 = LinearInterpolate(
   B=Multiply(A=Constant3Vector() {constant=<rgba 1 0.9201 0.5208 1>}, B=Power(Base=@n20.G) {constExponent=3})
   Alpha=ScalarParameter() {defaultValue=1, parameterName=AddedColorful}
  )
  n22 = ParticleColor()
  n23 = TextureCoordinate()
  n24 = CurveAtlasRowParameter(CurveTime=ComponentMask(None=@n23) {g=true}) {
   atlas=&/CAtlas_SinisterMelon
   curve=&/CV_SinisterMelon_SlashColor
   defaultValue=4
   parameterName=Color
  }
  n25 = CurveAtlasRowParameter(CurveTime=ComponentMask(None=@n23) {r=true}) {
   atlas=&CAtlas_SinisterMelon
   curve=&/CV_SinisterMelon_SlashColor2
   defaultValue=5
   parameterName=Color2
  }
  n26 = Multiply(A=@n24, B=@n25)
  n27 = ScalarParameter() {defaultValue=1, parameterName=MainColorIntensity}
  n28 = Multiply(A=@n22.RGB, B=Add(B=Multiply(A=@n26, B=@n27)))
  n29 = VectorParameter() {parameterName=Default_Tex_Panner, defaultValue=<rgba 0.5 0 0 0>}
  n30 = TextureSample(UVs=Panner(Coordinate=TextureCoordinate() {uTiling=0.5, vTiling=0.5}, Speed=ComponentMask(None=@n29.RGB) {g=true, r=true}) {speedX=0.5}) {samplerType=SAMPLERTYPE_LinearGrayscale, texture=/Game/Weapons/FORT_Melee/Pickaxe_Cube_Ninja_Male/FX/T_FluidFlameTile}
  n31 = LinearInterpolate(
   B=LinearInterpolate(B=ScalarParameter() {defaultValue=2, parameterName=Detail Noise Max Intensity}, Alpha=Power(Base=@n30.R) {constExponent=1}) {constA=0.7, constB=2}
   Alpha=ScalarParameter() {defaultValue=1, parameterName=UseDetail Noise}
  ) {constA=1}
  n32 = LinearInterpolate(A=Multiply(A=Multiply(A=@n28), B=@n31), B=Multiply(A=@n22.RGB, B=@n31), Alpha=ScalarParameter() {parameterName=Use Simple Color})
  n33 = ScalarParameter() {parameterName=Des}
  n34 = DynamicParameter() {parameterIndex=2, defaultValue=<rgba 10 1 0 1>, paramNames [Width,DesScale,none,Param4]}
  MaterialFunctionCall(Emissive=Desaturation(None=Add(A=@n21, B=@n32), Fraction=Multiply(A=@n33, B=@n34.DesScale))) {materialFunction=&MF_FX_EmissiveAdjustments}
  n35 = ComponentMask(None=TextureCoordinate()) {r=true}
  n36 = Power(Base=OneMinus(None=Sine(None=@n35) {period=2})) {constExponent=0.3}
  n37 = ComponentMask(None=TextureCoordinate()) {r=true}
  n38 = Constant() {r=0.7}
  n39 = Multiply(A=Time(), B=@n38)
  n40 = ComponentMask(None=Add(A=TextureCoordinate()) {constB=0}) {r=true}
  n41 = Multiply(A=Multiply(A=@n38, B=Time()) {constA=1}) {constB=-1}
  n42 = SmoothStep(Value=ComponentMask(None=TextureCoordinate()) {r=true}) {constMin=0.45}
  n43 = LinearInterpolate(
   A=Sine(None=Add(A=@n37, B=@n39)) {period=0.4}
   B=Sine(None=Add(A=@n40, B=@n41)) {period=0.4}
   Alpha=Saturate(None=Multiply(A=@n42) {constB=5})
  )
  n44 = VectorParameter() {parameterName=WPODirection, defaultValue=<rgba 0 1 0 0>}
  Multiply(
   A=Multiply(A=Multiply(A=@n36, B=@n43), B=Transform(None=@n44.RGB))
   B=DynamicParameter() {defaultValue=<rgba 0 0.2 0 10>, paramNames [Dissolve, DissolveOffset, MaskOffset, WPOIntensity]}.WPOIntensity
  )
  SphereMask(A=TextureCoordinate() {vTiling=0}, B=Constant2Vector() {r=0.5}) {attenuationRadius=0.5, hardnessPercent=0}
 ]
}
