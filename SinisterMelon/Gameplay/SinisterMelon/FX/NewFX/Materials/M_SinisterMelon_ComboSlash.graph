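# Translucent, unlit, two-sided material for the SinisterMelon combo slash effect, built for Niagara mesh particles.
# A camera-vector triplanar nebula texture blends with a curve-atlas gradient through the GalaxyLerp dynamic parameter, then particle color and brightness scale it.
# A slash-shaped mask, particle alpha, noise-based HeightLerp dissolve and view-angle fade form the opacity mask, which feeds depth fade and an index-of-refraction blend.
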Material /SinisterMelon/Gameplay/SinisterMelon/FX/NewFX/Materials/M_SinisterMelon_ComboSlash
graph {
 properties {
  bAutomaticallySetUsageInEditor=false
  bUsedWithNiagaraMeshParticles=true
  blendMode=BLEND_Translucent
  refractionMethod=RM_IndexOfRefraction
  shadingModel=MSM_Unlit
  twoSided=true
  usedShadingModels=Unlit
 }
 nodes [
  n1 = ParticleColor()
  n2 = TextureCoordinate()
  n3 = ComponentMask(None=VectorParameter() {parameterName=AlphaUVScale, defaultValue=<rgba 1 1.1 0 0>}.RGB) {g=true, r=true}
  n4 = VectorParameter() {parameterName=UVOffset, defaultValue=<rgba 0 -0.05 0 0>}
  n5 = TextureSampleParameter2D(UVs=Add(A=Multiply(A=@n2, B=@n3), B=ComponentMask(None=@n4.RGB) {g=true, r=true})) {
   parameterName=AlphaShape
   samplerSource=SSM_Clamp_WorldGroupSettings
   texture=&/T_SinisterMelon_SmallSlash
  }
  n6 = StaticComponentMaskParameter(None=@n5.RGB) {defaultR=true, parameterName=RGBMask}
  n7 = DynamicParameter() {defaultValue=<rgba 1 0.15 0 1>, paramNames [Param1,GalaxyLerp,Dissolve,Param4]}
  n8 = TextureCoordinate()
  n9 = ComponentMask(None=VectorParameter() {parameterName=DissUV, defaultValue=<rgba 0.7 0.7 0 0>}.RGB) {g=true, r=true}
  n10 = VectorParameter() {parameterName=Dissoffset, defaultValue=<rgba 0 0.3 0 0>}
  n11 = TextureSample(UVs=Add(A=Multiply(A=@n8, B=@n9), B=ComponentMask(None=@n10.RGB) {g=true, r=true})) {texture=&/T_CommonAuraNoise}
  n12 = ScalarParameter() {defaultValue=2, parameterName="Contrast (S)"}
  n13 = Multiply(
   A=Multiply(
    A=Multiply(A=@n1.A, B=@n6)
    B=MaterialFunctionCall(Transition Phase=@n7.Dissolve, Height Texture=@n11.R, Contrast=@n12) {materialFunction=&HeightLerp}.Results
   )
   B=MaterialFunctionCall("FadeSensitivity (optional)"=ScalarParameter() {defaultValue=1, parameterName="FadeSensitivity (optional) (S)"}) {materialFunction=&FadeBasedOnViewAngle}.Result
  )
  DepthFade(Opacity=@n13, FadeDistance=ScalarParameter() {defaultValue=100, parameterName=FadeDistance})
  LinearInterpolate(Alpha=@n13) {constA=1, constB=1.3}
  n14 = TextureObject() {texture=/Game/Characters/Player/Male/Medium/Bodies/M_Med_Soldier_04/Skins/Galaxy/Textures/T_FN_Nebula_style2}
  n15 = ScalarParameter() {defaultValue=3, parameterName=GalaxySize}
  n16 = Desaturation(
   None=Multiply(A=MaterialFunctionCall!(CameraVector, Texture=@n14, Tiling=@n15, Offset, AxisFadeContrast) {materialFunction=/Engine/Functions/Engine_MaterialFunctions01/Texturing/TriplanarCameraVector}.Result)
   Fraction=ScalarParameter() {parameterName=Des}
  )
  n17 = VectorParameter() {parameterName=GalaxyColor, defaultValue=<rgba 3.852 5 0.0038 1>}
  n18 = CurveAtlasRowParameter(CurveTime=ComponentMask(None=TextureCoordinate()) {r=true}) {
   atlas=&/CAtlas_SinisterMelon
   curve=&/CV_SinisterMelon_Color0
   parameterName=GradientColor
  }
  n19 = ComponentMask(None=TextureCoordinate()) {g=true}
  n20 = LinearInterpolate(
   A=VectorParameter() {parameterName=TopBottomColor, defaultValue=<rgba 2 1.3434 0.3939 1>}.RGB
   B=Multiply(A=@n18)
   Alpha=Power(Base=Sine(None=@n19) {period=2}) {constExponent=5}
  )
  n21 = Multiply(A=LinearInterpolate(
    A=Multiply(A=@n16, B=@n17.RGB)
    B=Multiply(B=@n20) {constA=1}
    Alpha=DynamicParameter() {defaultValue=<rgba 1 0.15 0 1>, paramNames [Param1,GalaxyLerp,Dissolve,Param4]}.GalaxyLerp
   ) {constAlpha=0.15})
  MaterialFunctionCall(Emissive=Multiply(A=Multiply(A=@n21, B=@n1.RGB), B=ScalarParameter() {defaultValue=5, parameterName=Brightness}) {constB=5}) {materialFunction=&MF_FX_EmissiveAdjustments}
  MaterialFunctionCall(
   Normal=TextureSample(UVs=TextureCoordinate() {uTiling=0.5, vTiling=5}) {samplerType=SAMPLERTYPE_Normal, texture=/Game/Accessories/FORT_Backpacks/Backpack_M_MED_EctoCat/FX/CloudNoise_T_N}.RGB
   Flatness=Constant() {r=-10}
  ) {materialFunction=&FlattenNormal}
  ScalarParameter() {defaultValue=0.15, parameterName=LerpGalaxy}
  TextureObject() {texture=/Game/Characters/Player/Male/Medium/Bodies/M_Med_Soldier_04/Skins/Galaxy/Textures/T_FN_Nebula_style_Green}
  TextureSample() {texture=&T_FN_Nebula_style2}
 ]
}
