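# Translucent deferred decal material for SinisterMelon effects that draws a cross-shaped mark, apparently a ground char.
# It samples a ground char texture in two stretched, offset layouts, one rotated by a rotator function, each shaped by a soft elongated sphere mask.
# The two arms merge with Max and are multiplied by decal lifetime opacity for fading.
# Scalar parameters supply base color, specular, metallic and roughness values.
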Material /SinisterMelon/Gameplay/SinisterMelon/FX/NewFX/Materials/M_SinisterMelon_DecalCross
graph {
 properties {
  bEnableResponsiveAA=true
  bOutputTranslucentVelocity=true
  bUsedWithNiagaraRibbons=true
  bUsedWithNiagaraSprites=true
  bUsedWithParticleSprites=true
  blendMode=BLEND_Translucent
  materialDomain=MD_DeferredDecal
  pixelDepthOffsetMode=PDOM_Legacy
  shadingModel=MSM_Unlit
  twoSided=true
  usedShadingModels=Unlit
 }
 nodes [
  n1 = Constant4Vector() {constant=<rgba 0.15 -2 0 0>}
  n2 = TextureSample(UVs=Add(A=TextureCoordinate() {uTiling=0.7, vTiling=5}, B=ComponentMask(None=@n1.RGB) {g=true, r=true})) {
   samplerSource=SSM_Clamp_WorldGroupSettings
   samplerType=SAMPLERTYPE_LinearColor
   texture=/Game/Effects/Fort_Effects/Gameplay/SmokeDustExplosions/Textures/T_GroundChar_01
  }
  n3 = Multiply(A=SphereMask(A=TextureCoordinate() {vTiling=4}, B=Constant2Vector() {g=2, r=0.5}) {attenuationRadius=0.45, hardnessPercent=0}, B=@n2.R)
  n4 = TextureCoordinate() {uTiling=5, vTiling=0.7}
  n5 = ScalarParameter() {defaultValue=0.25}
  n6 = Constant4Vector() {constant=<rgba -0.15 -2 0 0>}
  n7 = TextureSample(UVs=Add(
    A=MaterialFunctionCall(UVs=@n4, "Rotation Angle (0-1)"=@n5) {materialFunction=&CustomRotator}.Rotated Values
    B=ComponentMask(None=@n6.RGB) {g=true, r=true}
   )) {
   samplerSource=SSM_Clamp_WorldGroupSettings
   samplerType=SAMPLERTYPE_LinearColor
   texture=&T_GroundChar_01
  }
  n8 = Multiply(A=SphereMask(A=TextureCoordinate() {uTiling=4}, B=Constant2Vector() {g=0.5, r=2}) {attenuationRadius=0.45, hardnessPercent=0}, B=@n7.R)
  Multiply(A=DecalLifetimeOpacity().Opacity, B=Multiply(A=Max(A=@n3, B=@n8))) {constA=1}
  ScalarParameter() {parameterName=BaseColor}
  ScalarParameter() {defaultValue=0.5, parameterName=Specular}
  ScalarParameter() {parameterName=Metallic}
  ScalarParameter() {defaultValue=1, parameterName=Roughness}
 ]
}
