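# Unlit, translucent, two-sided Niagara mesh-particle material for the SinisterMelon impact core effect.
# Emissive blends triplanar camera-vector noise, nebula and spiral-blurred stars with a vertical color gradient, modulated by noise and particle color.
# Opacity combines a panner-driven shape mask, view-angle fade, smoke-texture HeightLerp dissolve and depth fade.
# Dynamic parameters drive dissolve and SplineThicken width; a curve-atlas gradient branch appears unused.
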Material /SinisterMelon/Gameplay/SinisterMelon/FX/NewFX/Materials/M_SinisterMelon_ImpactCore
graph {
 properties {
  bAutomaticallySetUsageInEditor=false
  bUsedWithNiagaraMeshParticles=true
  blendMode=BLEND_Translucent
  shadingModel=MSM_Unlit
  twoSided=true
  usedShadingModels=Unlit
 }
 nodes [
  n1 = ParticleColor()
  n2 = TextureCoordinate()
  n3 = StaticComponentMaskParameter(None=TextureSampleParameter2D(UVs=Panner(Coordinate=@n2)) {parameterName=AlphaShape, texture=&/T_SinisterMelon_Impact}.G) {defaultR=true, parameterName=RGBMask}
  n4 = Multiply(A=Multiply(A=Multiply(A=@n1.A, B=@n3)))
  n5 = MaterialFunctionCall("FadeSensitivity (optional)"=ScalarParameter() {defaultValue=1, parameterName="FadeSensitivity (optional) (S)"}) {materialFunction=&FadeBasedOnViewAngle}
  n6 = DynamicParameter() {defaultValue=<rgba 1 0.1 0 1>, paramNames [SplineWidth,GalaxyLerp,Dissolve,Param4]}
  n7 = TextureSample(UVs=TextureCoordinate() {vTiling=2}) {texture=&T_Inky_Smoke_Tile}
  n8 = Constant() {r=3}
  DepthFade(
   Opacity=Multiply(
    A=Multiply(A=@n4, B=@n5.Result)
    B=MaterialFunctionCall(Transition Phase=@n6.Dissolve, Height Texture=@n7.R, Contrast=@n8) {materialFunction=&HeightLerp}.Results
   )
   FadeDistance=ScalarParameter() {defaultValue=50, parameterName=FadeDistance}
  ) {fadeDistanceDefault=50}
  n9 = MaterialFunctionCall!(
   CameraVector
   Texture=TextureObject() {samplerType=SAMPLERTYPE_LinearGrayscale, texture=/Game/Effects/Fort_Effects/Emotes/UltraLight/Texture/T_Tiling_Noise_01_GreyScale}
   Tiling=ScalarParameter() {defaultValue=4, parameterName=GalaxySize}
   Offset
   AxisFadeContrast
  ) {materialFunction=/Engine/Functions/Engine_MaterialFunctions01/Texturing/TriplanarCameraVector}
  n10 = TextureObject() {texture=/Game/Characters/Player/Male/Medium/Bodies/M_Med_Soldier_04/Skins/Galaxy/Textures/T_FN_Nebula_style2}
  n11 = ScalarParameter() {defaultValue=4, parameterName=GalaxySize}
  n12 = LinearInterpolate(
   A=Multiply(A=Power(Base=@n9.Result)) {constB=2}
   B=Multiply(A=MaterialFunctionCall(Texture=@n10, Tiling=@n11) {materialFunction=&TriplanarCameraVector}.Result) {constB=2}
   Alpha=ScalarParameter() {defaultValue=0.5}
  )
  n13 = TextureCoordinate() {uTiling=15, vTiling=3}
  n14 = MaterialFunctionCall!(
   TextureObject=TextureObject() {texture=/Game/Weapons/WeaponSkins/Wraps/Textures/ShootingStars/ShootingStars_BackgroundStars_D}
   UV=@n13
   Distance=ScalarParameter() {defaultValue=0.02, parameterName="Distance (S)"}
   Distance Steps
   Radial Steps
   Radial Offset
   Kernel Power
  ) {materialFunction=/Engine/Functions/Engine_MaterialFunctions01/Texturing/SpiralBlur-Texture}
  n15 = Multiply(
   A=Add(
    A=Multiply(A=@n14.Result) {constB=0.6}
    B=TextureSample(UVs=@n13) {texture=&ShootingStars_BackgroundStars_D}.R
   ) {constB=0}
   B=ScalarParameter() {defaultValue=7, parameterName=StartIntensity}
  ) {constB=7}
  n16 = Desaturation(None=Multiply(A=Add(A=@n12, B=@n15)), Fraction=ScalarParameter() {defaultValue=0.3, parameterName=Des})
  n17 = VectorParameter() {parameterName=GalaxyColor, defaultValue=<rgba 1 0.397 0.0095 1>}
  n18 = ComponentMask(None=TextureCoordinate()) {g=true}
  n19 = LinearInterpolate(
   A=VectorParameter() {parameterName=TopBottomColor, defaultValue=<rgba 0.5521 0.5521 0.5521 1>}.RGB
   B=VectorParameter() {defaultValue=<rgba 1 0.6821 0.0759 1>}.RGB
   Alpha=Power(Base=Sine(None=@n18) {period=2}) {constExponent=5}
  )
  n20 = Multiply(A=LinearInterpolate(A=Multiply(A=@n16, B=@n17.RGB), B=Multiply(B=@n19) {constA=1}, Alpha=ScalarParameter() {defaultValue=0.15, parameterName=LerpGalaxy}) {constAlpha=0.15})
  n21 = LinearInterpolate(
   A=ScalarParameter() {parameterName=Param_1}
   Alpha=TextureSample(UVs=TextureCoordinate() {uTiling=2}) {samplerType=SAMPLERTYPE_Grayscale, texture=&LowResBlurredNoise}.R
  )
  n22 = Multiply(A=Multiply(A=Multiply(A=@n20, B=@n21)), B=@n1.RGB)
  n23 = ScalarParameter() {defaultValue=5, parameterName=Brightness}
  MaterialFunctionCall(Emissive=LinearInterpolate(A=Multiply(A=@n22, B=@n23) {constB=5}, B=ParticleColor().RGB, Alpha=ScalarParameter() {parameterName=UseSimpleGlow})) {materialFunction=&MF_FX_EmissiveAdjustments}
  MaterialFunctionCall() {materialFunction=&DebugTimeSine}
  n24 = Multiply(A=Constant() {r=20}, B=DynamicParameter() {paramNames [SplineWidth,Param2,Param3,Param4]}.SplineWidth)
  MaterialFunctionCall(WidthBase=@n24, WidthTip=@n24, Expand U or V UV Channel=StaticBool()) {materialFunction=&SplineThicken}
  n25 = CurveAtlasRowParameter(CurveTime=ComponentMask(None=TextureCoordinate()) {r=true}) {
   atlas=&/CAtlas_SinisterMelon
   curve=&/CV_SinisterMelon_Color1
   defaultValue=1
   parameterName=GradientColor
  }
  Multiply(A=@n25)
 ]
}
