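# Translucent, unlit, two-sided material for Niagara mesh particles that draws scrolling airflow streaks, by its name for the Sinister Melon projectile.
# A slowly panned blurred noise distorts the UVs of two faster-scrolling noise and smoke textures, which multiply into a streak pattern.
# Gradients along V, a view-angle fade and particle alpha shape that pattern, which feeds a depth fade for soft intersections.
# Particle color drives emissive through a shared adjustment function.
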Material /SinisterMelon/Gameplay/SinisterMelon/FX/NewFX/Materials/M_SinisterMelon_Projectile_Airflow
graph {
 properties {
  bAutomaticallySetUsageInEditor=false
  bUsedWithNiagaraMeshParticles=true
  blendMode=BLEND_Translucent
  shadingModel=MSM_Unlit
  twoSided=true
  usedShadingModels=Unlit
 }
 nodes [
  n1 = ParticleColor()
  n2 = VectorParameter() {parameterName=UVScale, defaultValue=<rgba 1 1 0 0>}
  n3 = Panner(Coordinate=Multiply(A=TextureCoordinate() {uTiling=3, vTiling=0.5}, B=ComponentMask(None=@n2.RGB) {g=true, r=true})) {speedY=0.1}
  n4 = Multiply(A=ComponentMask(None=TextureSample(UVs=@n3) {samplerType=SAMPLERTYPE_Grayscale, texture=&LowResBlurredNoise}.RGB) {g=true, r=true}) {constB=0.2}
  n5 = Add(A=TextureCoordinate() {uTiling=3, vTiling=0.35}, B=@n4)
  n6 = ComponentMask(None=VectorParameter() {parameterName=UVScale, defaultValue=<rgba 1 1 0 0>}.RGB) {g=true, r=true}
  n7 = TextureSampleParameter2D(UVs=Panner(Coordinate=Multiply(A=@n5, B=@n6)) {speedY=1}) {parameterName=TexA, texture=/Game/Weapons/FORT_Melee/Pickaxe_CoatCheck/FX/T_Pickaxe_CoatCheck_Noise}
  n8 = Add(A=TextureCoordinate() {uTiling=3, vTiling=0.5}, B=@n4)
  n9 = TextureSampleParameter2D(UVs=Panner(Coordinate=Multiply(A=@n8, B=@n6)) {speedY=1}) {parameterName=TexB, texture=&T_Inky_Smoke_Tile}
  n10 = Power(Base=Multiply(A=@n7.R, B=@n9.R), Exp=ScalarParameter() {defaultValue=1, parameterName=Power}) {constExponent=1}
  n11 = TextureCoordinate()
  n12 = TextureCoordinate()
  n13 = Power(Base=OneMinus(None=ComponentMask(None=@n12) {g=true}))
  n14 = Multiply(A=Power(Base=ComponentMask(None=@n11) {g=true}), B=Saturate(None=Multiply(A=@n13) {constB=10}))
  n15 = MaterialFunctionCall("FadeSensitivity (optional)"=Constant() {r=3}) {materialFunction=&FadeBasedOnViewAngle}
  n16 = Multiply(A=Multiply(A=Multiply(A=@n10, B=@n14)) {constB=2}, B=Saturate(None=Multiply(A=@n15.Result) {constB=5}))
  DepthFade(Opacity=Multiply(A=@n1.A, B=Saturate(None=@n16))) {fadeDistanceDefault=300}
  MaterialFunctionCall(Emissive=@n1.RGB) {materialFunction=&MF_FX_EmissiveAdjustments}
  Power(Base=@n7.R) {constExponent=1}
  Power(Base=@n9.R) {constExponent=1}
 ]
}
