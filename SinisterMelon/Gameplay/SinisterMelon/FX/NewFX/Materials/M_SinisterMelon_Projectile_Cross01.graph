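# Translucent, unlit, two-sided Niagara mesh-particle material for the SinisterMelon projectile effect.
# Emissive blends a camera-vector galaxy of nebula, noise, rainbow and blurred stars with a curve-atlas gradient via dynamic GalaxyLerp, scaled by particle color.
# Opacity pans a noise-distorted shape texture, masked by tail, view-angle and depth fades plus a dynamic-parameter side-hide gradient in local space.
# A time-driven sine along U builds a vertex wobble in a local direction.
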Material /SinisterMelon/Gameplay/SinisterMelon/FX/NewFX/Materials/M_SinisterMelon_Projectile_Cross01
graph {
 properties {
  bAutomaticallySetUsageInEditor=false
  bUsedWithNiagaraMeshParticles=true
  blendMode=BLEND_Translucent
  shadingModel=MSM_Unlit
  twoSided=true
  usedShadingModels=Unlit
 }
 nodes [
  n1 = MaterialFunctionCall!(
   CameraVector
   Texture=TextureObject() {samplerType=SAMPLERTYPE_LinearGrayscale, texture=/Game/Effects/Fort_Effects/Emotes/UltraLight/Texture/T_Tiling_Noise_01_GreyScale}
   Tiling=ScalarParameter() {defaultValue=3, parameterName=GalaxySize}
   Offset
   AxisFadeContrast
  ) {materialFunction=/Engine/Functions/Engine_MaterialFunctions01/Texturing/TriplanarCameraVector}
  n2 = TextureObject() {texture=/Game/Characters/Player/Male/Medium/Bodies/M_Med_Soldier_04/Skins/Galaxy/Textures/T_FN_Nebula_style2}
  n3 = ScalarParameter() {defaultValue=3, parameterName=GalaxySize}
  n4 = LinearInterpolate(
   A=Multiply(A=Power(Base=@n1.Result)) {constB=2}
   B=Multiply(A=MaterialFunctionCall(Texture=@n2, Tiling=@n3) {materialFunction=&TriplanarCameraVector}.Result) {constB=2}
   Alpha=ScalarParameter() {defaultValue=0.5}
  )
  n5 = Add(
   A=TextureCoordinate() {uTiling=0.5, vTiling=0.5}
   B=ComponentMask(None=VectorParameter() {defaultValue=<rgba 0 -0.35 0 0>}.RGB) {g=true, r=true}
  )
  n6 = Multiply(A=Desaturation(
    None=TextureSample(UVs=@n5) {texture=&/T_RainbowColor03}.RGB
    Fraction=ScalarParameter() {defaultValue=-1, parameterName=Param_2}
   )) {constB=1.5}
  n7 = TextureObject() {texture=/Game/Weapons/WeaponSkins/Wraps/Textures/ShootingStars/ShootingStars_BackgroundStars_D}
  n8 = TextureCoordinate() {uTiling=15, vTiling=3}
  n9 = ScalarParameter() {defaultValue=0.02, parameterName="Distance (S)"}
  n10 = Add(
   A=Multiply(A=MaterialFunctionCall!(TextureObject=@n7, UV=@n8, Distance=@n9, Distance Steps, Radial Steps, Radial Offset, Kernel Power) {materialFunction=/Engine/Functions/Engine_MaterialFunctions01/Texturing/SpiralBlur-Texture}.Result) {constB=0.6}
   B=TextureSample(UVs=@n8) {texture=&ShootingStars_BackgroundStars_D}.R
  ) {constB=0}
  n11 = ScalarParameter() {defaultValue=7, parameterName=StartIntensity}
  n12 = Multiply(A=Add(A=Multiply(A=@n4, B=@n6), B=Multiply(A=@n10, B=@n11) {constB=7})) {constB=0.2}
  n13 = ScalarParameter() {defaultValue=0.3, parameterName=Des}
  n14 = VectorParameter() {parameterName=TopBottomColor, defaultValue=<rgba 0.5521 0.5521 0.5521 1>}
  n15 = CurveAtlasRowParameter(CurveTime=ComponentMask(None=TextureCoordinate()) {r=true}) {
   atlas=&/CAtlas_SinisterMelon
   curve=&/CV_SinisterMelon_Color1
   defaultValue=1
   parameterName=GradientColor
  }
  n16 = Multiply(A=@n15)
  n17 = TextureCoordinate()
  n18 = Power(Base=Sine(None=ComponentMask(None=@n17) {g=true}) {period=2}) {constExponent=5}
  n19 = LinearInterpolate(
   A=Multiply(A=Desaturation(None=@n12, Fraction=@n13), B=VectorParameter() {parameterName=GalaxyColor, defaultValue=<rgba 1 0.397 0.0095 1>}.RGB)
   B=Multiply(B=LinearInterpolate(A=@n14.RGB, B=@n16, Alpha=@n18)) {constA=1}
   Alpha=DynamicParameter() {defaultValue=<rgba 1 0.1 0 0>, paramNames [Param1,GalaxyLerp,Dissolve,Param4]}.GalaxyLerp
  ) {constAlpha=0.15}
  n20 = TextureSample() {samplerType=SAMPLERTYPE_Grayscale, texture=&LowResBlurredNoise}
  n21 = TextureCoordinate()
  n22 = Multiply(A=OneMinus(None=ComponentMask(None=@n21) {r=true})) {constB=1.5}
  n23 = Multiply(
   A=Multiply(A=Multiply(A=@n19), B=LinearInterpolate(Alpha=@n20.R) {constA=0.5})
   B=LinearInterpolate(
    A=ScalarParameter() {defaultValue=0.5, parameterName=CenterMin}
    B=ScalarParameter() {defaultValue=1.7, parameterName=CenterMax}
    Alpha=Saturate(None=@n22)
   ) {constA=0.5}
  )
  n24 = ParticleColor()
  MaterialFunctionCall(Emissive=Multiply(A=Multiply(A=@n23, B=@n24.RGB), B=ScalarParameter() {defaultValue=5, parameterName=Brightness}) {constB=5}) {materialFunction=&MF_FX_EmissiveAdjustments}
  MaterialFunctionCall(
   Texture=TextureObject() {texture=&T_RainbowColor03}
   Tiling=ScalarParameter() {defaultValue=3, parameterName=Param_1}
  ) {materialFunction=&TriplanarCameraVector}
  n25 = LinearInterpolate(Alpha=VertexColor().R) {constB=0.4}
  n26 = ComponentMask(None=TextureCoordinate()) {r=true}
  n27 = LinearInterpolate(Alpha=Saturate(None=Multiply(A=@n26) {constB=0.5}))
  n28 = TextureCoordinate()
  n29 = ComponentMask(None=VectorParameter() {parameterName=AlphaUVScale, defaultValue=<rgba 0.3 1 0 0>}.RGB) {g=true, r=true}
  n30 = TextureSample(UVs=Panner(Coordinate=TextureCoordinate() {uTiling=0.75, vTiling=0.75}) {speedX=-0.5, speedY=0.5}) {texture=/Game/Weapons/FORT_Melee/Pickaxe_CoatCheck/FX/T_Pickaxe_CoatCheck_Noise}
  n31 = ScalarParameter() {defaultValue=0.15, parameterName=NoiseDisturb}
  n32 = Add(A=Add(A=Multiply(A=@n28, B=@n29), B=Multiply(A=@n30.R, B=@n31) {constB=0.15}), B=AppendVector(A=Multiply(A=@n25, B=@n27), B=Constant()))
  n33 = StaticComponentMaskParameter(None=TextureSampleParameter2D(UVs=Panner(Coordinate=@n32) {speedX=-0.8}) {parameterName=AlphaShape, texture=&/T_SinisterMelon_ProjectileCenter}.RGB) {defaultR=true, parameterName=RGBMask}
  n34 = ComponentMask(None=TextureCoordinate()) {r=true}
  n35 = Multiply(A=Power(Base=OneMinus(None=@n34), Exp=ScalarParameter() {defaultValue=2, parameterName=TailMaskPower})) {constB=10}
  n36 = Multiply(A=Multiply(A=Multiply(A=@n24.A, B=@n33), B=Saturate(None=@n35)))
  n37 = MaterialFunctionCall("FadeSensitivity (optional)"=ScalarParameter() {defaultValue=1, parameterName="FadeSensitivity (optional) (S)"}) {materialFunction=&FadeBasedOnViewAngle}
  Multiply(A=NamedRerouteUsage(), B=DepthFade(Opacity=Multiply(A=@n36, B=@n37.Result)) {fadeDistanceDefault=400})
  n38 = ComponentMask(None=TextureCoordinate()) {r=true}
  n39 = Multiply(A=Power(Base=@n38) {constExponent=1}) {constB=5}
  n40 = Add(A=ComponentMask(None=TextureCoordinate()) {r=true}, B=Multiply(A=Time(), B=Constant() {r=-2}))
  Multiply(
   A=Multiply(A=Saturate(None=@n39), B=Sine(None=@n40) {period=0.4})
   B=Transform(None=VectorParameter() {parameterName=WPODirection, defaultValue=<rgba 0 3 0 0>}.RGB) {transformSourceType=TRANSFORMSOURCE_Local}
  )
  n41 = DynamicParameter() {parameterIndex=3, defaultValue=<rgba 0 0 0 0>, paramNames [Left,Right,Param3,Param4]}
  n42 = Add(A=@n41.Left, B=@n41.Left)
  n43 = Constant()
  MaterialExpressionVertexInterpolator_0 = VertexInterpolator(VS=LocalPosition() {includedOffsets=ExcludeOffsets, localOrigin=InstancePreSkinning}.XY)
  n44 = Convert(None=@MaterialExpressionVertexInterpolator_0.PS) {
   nodeName=Break Float3
   convertInputs [{
     type=Vector3
     expressionInput {
      expression=&self:MaterialExpressionVertexInterpolator_0
      inputName=None
      mask=0
      maskA=0
      maskB=0
      maskG=0
      maskR=0
      outputIndex=0
     }
    }]
   convertOutputs [{type=Scalar}, {type=Scalar}, {type=Scalar}]
  }
  n45 = ScalarParameter() {defaultValue=8.8172, parameterName=HideGradient}
  n46 = Clamp(None=Divide(A=@n44, B=@n45) {constB=1}) {clampMode=CMODE_ClampMin}
  n47 = Saturate(None=Multiply(A=@n46, B=@n46))
  n48 = Clamp(None=Divide(A=Multiply(A=@n44) {constB=-1}, B=@n45) {constB=1}) {clampMode=CMODE_ClampMin}
  n49 = If(A=@n41.Left, "A > B"=@n47, "A == B"=@n47, "A < B"=Saturate(None=Multiply(A=@n48, B=@n48))) {constB=1}
  n50 = Constant() {r=1}
  NamedRerouteDeclaration(None=If(A=@n42, "A > B"=@n43, "A == B"=@n43, "A < B"=If(A=@n42, "A > B"=@n49, "A == B"=@n50, "A < B"=@n50)) {constB=2}) {name=Hide, nodeColor=<rgba 1 0.3984 0 1>}
  OneMinus(None=Sine(None=@n38) {period=2})
  PreSkinnedPosition()
  ScalarParameter() {defaultValue=3, parameterName=GalaxySize}
  ScalarParameter() {defaultValue=0.15, parameterName=LerpGalaxy}
  TextureObject() {texture=/Game/Characters/Player/Male/Medium/Bodies/M_Med_Soldier_04/Skins/Galaxy/Textures/T_FN_Nebula_style_Green}
  TextureSample() {texture=&T_FN_Nebula_style2}
 ]
}
