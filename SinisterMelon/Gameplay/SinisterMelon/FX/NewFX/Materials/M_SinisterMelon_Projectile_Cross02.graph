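# Translucent, unlit, two-sided Niagara mesh particle material for a Sinister Melon projectile effect.
# Opacity merges two mirrored, panning impact-texture samples, erodes them through a smoke-noise HeightLerp and softens intersections with DepthFade.
# Emissive blends a camera-vector triplanar galaxy with blurred stars toward a curve-atlas gradient, boosted at the center.
# Dynamic parameters drive that blend, the dissolve and a SplineThicken width tapered by a sphere mask.
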Material /SinisterMelon/Gameplay/SinisterMelon/FX/NewFX/Materials/M_SinisterMelon_Projectile_Cross02
graph {
 properties {
  bAutomaticallySetUsageInEditor=false
  bUsedWithNiagaraMeshParticles=true
  blendMode=BLEND_Translucent
  shadingModel=MSM_Unlit
  twoSided=true
  usedShadingModels=Unlit
 }
 nodes [
  ComponentMask(None=TextureCoordinate()) {r=true}
  n1 = Multiply(A=TextureCoordinate(), B=ComponentMask(None=VectorParameter() {parameterName=UVScale, defaultValue=<rgba 1.5 1 0 0>}.RGB) {g=true, r=true})
  n2 = TextureSampleParameter2D(UVs=Panner(Coordinate=@n1) {speedX=-0.8}) {parameterName=AlphaShape, texture=&/T_SinisterMelon_Impact}
  n3 = ComponentMask(None=TextureCoordinate()) {r=true}
  n4 = Constant() {r=50}
  n5 = Multiply(A=SmoothStep(Value=@n3) {constMin=0.4}, B=@n4) {constB=50}
  n6 = Constant2Vector() {g=1, r=-1}
  n7 = TextureSampleParameter2D(UVs=Panner(Coordinate=Multiply(A=@n1, B=@n6)) {speedX=-0.8}) {parameterName=AlphaShape, texture=&T_SinisterMelon_Impact}
  n8 = Multiply(A=SmoothStep(Value=OneMinus(None=@n3)) {constMin=0.4}, B=@n4) {constB=50}
  n9 = Max(
   A=Multiply(A=StaticComponentMaskParameter(None=@n2.RGB) {defaultR=true, parameterName=RGBMask}, B=Saturate(None=@n5))
   B=Multiply(A=StaticComponentMaskParameter(None=@n7.RGB) {defaultR=true, parameterName=RGBMask}, B=Saturate(None=@n8))
  )
  n10 = MaterialFunctionCall(
   Transition Phase=DynamicParameter() {defaultValue=<rgba 1 0.1 0 1>, paramNames [SplineWidth,GalaxyLerp,Dissolve,Param4]}.Dissolve
   Height Texture=TextureSample(UVs=TextureCoordinate() {vTiling=2}) {texture=&T_Inky_Smoke_Tile}.R
   Contrast=Constant() {r=3}
  ) {materialFunction=&HeightLerp}
  DepthFade(
   Opacity=Multiply(A=ParticleColor().A, B=Multiply(A=@n9, B=@n10.Results))
   FadeDistance=ScalarParameter() {defaultValue=50, parameterName=FadeDistance}
  ) {fadeDistanceDefault=50}
  n11 = MaterialFunctionCall!(
   CameraVector
   Texture=TextureObject() {samplerType=SAMPLERTYPE_LinearGrayscale, texture=/Game/Effects/Fort_Effects/Emotes/UltraLight/Texture/T_Tiling_Noise_01_GreyScale}
   Tiling=ScalarParameter() {defaultValue=3, parameterName=GalaxySize}
   Offset
   AxisFadeContrast
  ) {materialFunction=/Engine/Functions/Engine_MaterialFunctions01/Texturing/TriplanarCameraVector}
  n12 = TextureObject() {texture=/Game/Characters/Player/Male/Medium/Bodies/M_Med_Soldier_04/Skins/Galaxy/Textures/T_FN_Nebula_style2}
  n13 = ScalarParameter() {defaultValue=3, parameterName=GalaxySize}
  n14 = LinearInterpolate(
   A=Multiply(A=Power(Base=@n11.Result)) {constB=2}
   B=Multiply(A=MaterialFunctionCall(Texture=@n12, Tiling=@n13) {materialFunction=&TriplanarCameraVector}.Result) {constB=2}
   Alpha=Constant() {r=0.5}
  )
  n15 = Add(
   A=TextureCoordinate() {uTiling=0.5, vTiling=0.5}
   B=ComponentMask(None=VectorParameter() {defaultValue=<rgba 0 -0.35 0 0>}.RGB) {g=true, r=true}
  )
  n16 = Multiply(A=Desaturation(
    None=TextureSample(UVs=@n15) {texture=&/T_RainbowColor03}.RGB
    Fraction=Constant() {r=-1}
   )) {constB=1.5}
  n17 = TextureObject() {texture=/Game/Weapons/WeaponSkins/Wraps/Textures/ShootingStars/ShootingStars_BackgroundStars_D}
  n18 = TextureCoordinate() {uTiling=15, vTiling=3}
  n19 = ScalarParameter() {defaultValue=0.02, parameterName="Distance (S)"}
  n20 = Add(
   A=Multiply(A=MaterialFunctionCall!(TextureObject=@n17, UV=@n18, Distance=@n19, Distance Steps, Radial Steps, Radial Offset, Kernel Power) {materialFunction=/Engine/Functions/Engine_MaterialFunctions01/Texturing/SpiralBlur-Texture}.Result) {constB=0.6}
   B=TextureSample(UVs=@n18) {texture=&ShootingStars_BackgroundStars_D}.R
  ) {constB=0}
  n21 = ScalarParameter() {parameterName=StartIntensity}
  n22 = Multiply(A=Add(A=Multiply(A=@n14, B=@n16), B=Multiply(A=@n20, B=@n21) {constB=7})) {constB=0.2}
  n23 = ScalarParameter() {defaultValue=0.3, parameterName=Des}
  n24 = VectorParameter() {parameterName=TopBottomColor, defaultValue=<rgba 0.5521 0.5521 0.5521 1>}
  n25 = SphereMask(A=TextureCoordinate() {vTiling=0}, B=Constant2Vector() {r=0.5}) {attenuationRadius=0.5, hardnessPercent=0}
  n26 = CurveAtlasRowParameter(CurveTime=Power(Base=OneMinus(None=@n25))) {
   atlas=&/CAtlas_SinisterMelon
   curve=&/CV_SinisterMelon_Color1
   defaultValue=1
   parameterName=GradientColor
  }
  n27 = Multiply(A=@n26)
  n28 = TextureCoordinate()
  n29 = Power(Base=Sine(None=ComponentMask(None=@n28) {g=true}) {period=2}) {constExponent=5}
  n30 = LinearInterpolate(
   A=Multiply(A=Desaturation(None=@n22, Fraction=@n23), B=VectorParameter() {parameterName=GalaxyColor, defaultValue=<rgba 1 0.397 0.0095 1>}.RGB)
   B=Multiply(B=LinearInterpolate(A=@n24.RGB, B=@n27, Alpha=@n29)) {constA=1}
   Alpha=DynamicParameter() {defaultValue=<rgba 1 0.1 0 1>, paramNames [SplineWidth,GalaxyLerp,Dissolve,Param4]}.GalaxyLerp
  ) {constAlpha=0.15}
  n31 = TextureSample() {samplerType=SAMPLERTYPE_Grayscale, texture=&LowResBlurredNoise}
  n32 = TextureCoordinate() {vTiling=0}
  n33 = Constant2Vector() {r=0.5}
  n34 = Multiply(A=Power(Base=SphereMask(A=@n32, B=@n33) {attenuationRadius=0.5, hardnessPercent=0}))
  n35 = Multiply(
   A=Multiply(A=Multiply(A=@n30), B=LinearInterpolate(Alpha=@n31.R) {constA=0.5})
   B=LinearInterpolate(
    A=ScalarParameter() {defaultValue=0.7, parameterName=CenterMin}
    B=ScalarParameter() {defaultValue=3, parameterName=CenterMax}
    Alpha=Saturate(None=@n34)
   ) {constA=0.5}
  )
  n36 = ParticleColor()
  MaterialFunctionCall(Emissive=Multiply(A=Multiply(A=@n35, B=@n36.RGB), B=ScalarParameter() {defaultValue=5, parameterName=Brightness}) {constB=5}) {materialFunction=&MF_FX_EmissiveAdjustments}
  MaterialFunctionCall() {materialFunction=&DebugTimeSine}
  n37 = Multiply(A=Constant() {r=30}, B=DynamicParameter() {defaultValue=<rgba 1 0.1 0 1>, paramNames [SplineWidth,GalaxyLerp,Dissolve,Param4]}.SplineWidth)
  n38 = Power(Base=SphereMask(A=TextureCoordinate() {vTiling=0}, B=Constant2Vector() {r=0.5}) {attenuationRadius=0.5, hardnessPercent=0}) {constExponent=0.8}
  MaterialFunctionCall(WidthBase=Multiply(A=@n37, B=@n38), WidthTip=Multiply(A=@n37, B=@n38), Expand U or V UV Channel=StaticBool()) {materialFunction=&SplineThicken}
  TextureCoordinate() {uTiling=-1.5}
 ]
}
