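# Translucent, unlit, two-sided glow material for Sinister Melon effects on Niagara mesh particles.
# Particle color RGB feeds the MF_FX_EmissiveAdjustments function for emissive.
# Particle alpha is multiplied by a powered radial UV sphere mask, then depth-faded by a FadeDistance parameter to give soft opacity.
# A SplineWidth dynamic parameter, scaled by a constant, drives both base and tip widths of SplineThicken.
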Material /SinisterMelon/Gameplay/SinisterMelon/FX/NewFX/Materials/M_SinisterMelon_SimpleGlow
graph {
 properties {
  bAutomaticallySetUsageInEditor=false
  bUsedWithNiagaraMeshParticles=true
  blendMode=BLEND_Translucent
  shadingModel=MSM_Unlit
  twoSided=true
  usedShadingModels=Unlit
 }
 nodes [
  n1 = ParticleColor()
  n2 = SphereMask(A=TextureCoordinate(), B=Constant() {r=0.5}) {attenuationRadius=0.5, hardnessPercent=0}
  DepthFade(Opacity=Multiply(A=@n1.A, B=Power(Base=@n2) {constExponent=3}), FadeDistance=ScalarParameter() {defaultValue=50, parameterName=FadeDistance}) {fadeDistanceDefault=50}
  n3 = Multiply(A=Constant() {r=30}, B=DynamicParameter() {paramNames [SplineWidth,Param2,Param3,Param4]}.SplineWidth)
  MaterialFunctionCall(WidthBase=@n3, WidthTip=@n3, Expand U or V UV Channel=StaticBool()) {materialFunction=&SplineThicken}
  MaterialFunctionCall(Emissive=@n1.RGB) {materialFunction=&MF_FX_EmissiveAdjustments}
 ]
}
