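# Niagara system for a SinisterMelon charge-up effect, by its name a post-process volume (PPV) shown only to the local player.
# User.IsLocalPlayer sets the burst spawn probability to 1 or 0 for a two-second particle drawn through a Component Renderer.
# A curve ramps Particles.SpeedNormalizedB from zero early in life, then holds it.
# Particles.ComponentsEnabled compares age to lifetime minus delta time, toggling the component around the particle's final frame.
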NiagaraSystem /SinisterMelon/Gameplay/SinisterMelon/FX/NewFX/NS_SinisterMelon_ChargeUpPPV
dependencies {usedDataInterfaces [&NiagaraDataInterfaceCurve], usedDynamicInputs [], usedModules [&SystemState]}
emitters {PPV {
  emitterUpdateScript [
   &EmitterState(Life Cycle Mode=::NewEnumerator1, Loop Behavior=::NewEnumerator1, Loop Duration=2, UseLoopDelay=false)
   &SpawnBurst_Instantaneous(
    Spawn Probability=&MakeCustomFloatFromBool(Bool=$User.IsLocalPlayer, True Float=1, False Float=0)
    Use Spawn Probability=true
    UseLoopCountLimit=false
   )
  ]
  particleSpawnScript [&InitializeParticle(
    Color Mode=::NewEnumerator0
    Lifetime=2
    Sprite Size Mode=::NewEnumerator3
    Uniform Sprite Size=200
    UsePositionOffset=false
    Write Material Random=true
   )]
  particleUpdateScript [
   /SetVariables(Particles.SpeedNormalizedB:NiagaraFloat=&FloatFromCurve(FloatCurve=DataInterface() {propertyValues {Curve {keys [
         {interpMode=RCIM_Cubic, tangentMode=RCTM_User, time=0, value=0}
         {arriveTangent=0.5714, interpMode=RCIM_Cubic, leaveTangent=0.5714, time=0.2, value=0.4}
         <at 0.7 0.4 RCIM_Cubic>
         {interpMode=RCIM_Cubic, tangentMode=RCTM_User, time=1, value=0.4}
        ]}}})) {isSetParameters=true}
   &ParticleState()
   /SetVariables(Particles.ComponentsEnabled:NiagaraBool=&SetBoolByFloatComparison(
     A=$Particles.Age
     B=&Subtract_Float(A=$Particles.Lifetime, B=$Engine.DeltaTime)
     Comparison Type=::NewEnumerator6
    )) {isSetParameters=true}
  ]
  rendererClasses [&NiagaraComponentRendererProperties]
  renderers [{rendererClass=&NiagaraComponentRendererProperties, rendererIndex=0}]
 }}
summary {
 systemName=NS_SinisterMelon_ChargeUpPPV
 emitters [{emitterName=PPV, rendererClasses [&NiagaraComponentRendererProperties]}]
 userVariables [User.IsLocalPlayer {type=NiagaraBool, defaultValue=true}]
}
