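# Niagara system for the SinisterMelon combo-end AOE effect: a gold-orange moving head with trail, glow and sparks.
# System-level variables drive shared lifetime, velocity, drag, size and camera offset for the HeadMesh, MeshTrail and Glow bursts.
# HeadMesh layers five mesh renderers, writing separate color, scale, position and material attributes to vary the layers.
# Sparks spawn continuously from HeadMesh particles through an attribute reader, then scatter with cone velocity and curl noise.
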NiagaraSystem /SinisterMelon/Gameplay/SinisterMelon/FX/NewFX/NS_SinisterMelon_ComboEndAOE
dependencies {
 usedDataInterfaces [
  &NiagaraDataInterfaceArrayFloat3
  &NiagaraDataInterfaceVectorCurve
  &NiagaraDataInterfaceCurve
  &NiagaraDataInterfaceParticleRead
 ]
 usedDynamicInputs [&FN_LOD_Distance_Remap]
 usedModules [
  &self:SystemSpawnScript.SystemScriptSource.SystemScriptGraph.NiagaraNodeAssignment_1.SetVariables_#1
  &self:SystemSpawnScript.SystemScriptSource.SystemScriptGraph.NiagaraNodeAssignment_3.SetVariables_#2
  &self:SystemSpawnScript.SystemScriptSource.SystemScriptGraph.NiagaraNodeAssignment_4.SetVariables_#3
  &SystemState
  &self:SystemSpawnScript.SystemScriptSource.SystemScriptGraph.NiagaraNodeAssignment_0.SetVariables_#4
  &self:SystemSpawnScript.SystemScriptSource.SystemScriptGraph.NiagaraNodeAssignment_2.SetVariables_#5
  &self:SystemSpawnScript.SystemScriptSource.SystemScriptGraph.NiagaraNodeAssignment_5.SetVariables_#6
 ]
}
emitters {
 Glow {
  emitterUpdateScript [
   &EmitterState(Life Cycle Mode=::NewEnumerator1, Loop Behavior=::NewEnumerator1, Loop Delay=0.1, Loop Duration=0.6, UseLoopDelay=false)
   &SpawnBurst_Instantaneous(Spawn Count=1, Use Spawn Probability=false)
  ]
  particleSpawnScript [
   &InitializeParticle(
    Color=<rgba 0.5052 0.3948 0.0077 0.5>
    Lifetime=$System.LifetimeAll
    Position Offset=<0 10 0>
    Sprite Size=<400 400>
    Sprite Size Mode=::NewEnumerator4
    UsePositionOffset=true
    Write Material Random=true
   )
   &DynamicMaterialParameters(
    Index 0 Param 1=4
    Index 0 Param 2=30
    Index 0 Param 3=1
    Index 0 Param 4=1
    Module.Param0WriteEnabled=true
    Module.Param1WriteEnabled=true
    Module.Param2WriteEnabled=true
    Module.Param3WriteEnabled=true
   )
   &CameraOffset(Camera Offset Amount=&MakeCustomFloatFromBool(Bool=$User.bIsFrontEndPreview, True Float=1, False Float=30))
   &AddVelocity(Velocity=$System.VelocityAll)
  ]
  particleUpdateScript [
   &ParticleState()
   &ScaleSpriteSize(Scale Sprite Size Mode=::NewEnumerator0, Uniform Scale Factor=$System.SizeScale02)
   &ScaleColor(Scale Alpha=&Multiply_Float(A=0.7, B=1), ScaleA=true, ScaleRGB=true)
   &SpriteFacingAndAlignment(Sprite Facing=<0 0 1>, WriteAlignment=true, WriteFacing=true)
   &Drag(Drag=$System.DragAll, Use Linear Drag=true, Use Rotational Drag=true)
   &SolveForcesAndVelocity(Clamp Velocity=false, Limit Acceleration=false)
   &CameraOffset(Camera Offset Amount=$System.CameraOffsetAll) {module=CameraOffset001}
  ]
  rendererClasses [&NiagaraSpriteRendererProperties]
  renderers [{rendererClass=&NiagaraSpriteRendererProperties}]
 }
 HeadMesh {
  emitterUpdateScript [
   &EmitterState(Life Cycle Mode=::NewEnumerator1, Loop Behavior=::NewEnumerator1, Loop Duration=0.6, UseLoopDelay=false)
   &SpawnBurst_Instantaneous(Use Spawn Probability=false, UseLoopCountLimit=false)
  ]
  particleSpawnScript [
   &PartitionParticles()
   &InitializeParticle(
    Color=<rgba 0.5 0.3249 0.0078 0.7>
    Lifetime=$System.LifetimeAll
    Mesh Scale Mode=::NewEnumerator3
    Mesh Uniform Scale=1.5
    Position Offset=<0 50 0>
    UsePositionOffset=true
    Write Material Random=true
   )
   &InitialMeshOrientation(
    Mesh Orientation Mode=::NewEnumerator2
    Rotation=&SelectVectorFromArray(
     Direct Array Index=$Particles.PartitionParticles.Partition
     Vector Selection Array=DataInterface() {propertyValues {FloatData [<0 0 0>, <0 0.25 0>]}}
    )
    Use Rotation Vector=true
   )
   &AddVelocity(Velocity=$System.VelocityAll)
  ]
  particleUpdateScript [
   &ParticleState()
   &ScaleMeshSize(Scale Factor=<1.3 1.3 1>) {module=ScaleMeshSize001}
   &ScaleMeshSize(Scale Factor=$System.SizeAll)
   &DynamicMaterialParameters(
    Index 0 Param 2=0.2
    Module.Index0Float0Write=true
    Module.Index0Float1Write=true
    Module.Index0Float2Write=true
    Module.Index0Float3Write=true
   )
   &Drag(Drag=$System.DragAll, Use Linear Drag=true, Use Rotational Drag=true)
   &SolveForcesAndVelocity(Clamp Velocity=false, Limit Acceleration=false)
   &FN_Mesh_Attributes(
    Color=&MakeLinearColorFromVectorAndFloat(
     "Vector (RGB)"=&Multiply_Vector(A=&MakeVectorFromLinearColorRGB(Color=<rgba 1.5 1.2683 0.7351 2>), B=<1 1 1>)
     "Float (Alpha)"=&Multiply_Float(A=1, B=1)
    )
    Create Color Attribute=true
    Create Scale Attribute=true
    Scale=&Multiply_VectorByFloat(Vector=&Multiply_Vector(A=<1.1 1.1 0.8>, B=$System.SizeAll), Float=1.6)
    Write to Module Namespace=true
   )
   &FN_Mesh_Attributes(
    Color=&MakeLinearColorFromVectorAndFloat(
     "Vector (RGB)"=&Multiply_Vector(A=&MakeVectorFromLinearColorRGB(Color=<rgba 1.7 1.0359 0.4334 1.5>), B=<1 1 1>)
     "Float (Alpha)"=&Multiply_Float(A=1.5, B=1)
    )
    Create Color Attribute=true
    Create Scale Attribute=true
    Scale=&Multiply_VectorByFloat(Vector=&Multiply_Vector(A=<1 1 1>, B=$System.SizeAll), Float=1.5)
    Write to Module Namespace=true
   ) {module=FN_Mesh_Attributes001}
   &FN_Mesh_Attributes(
    Color=&MakeLinearColorFromVectorAndFloat(
     "Vector (RGB)"=&Multiply_Vector(A=&MakeVectorFromLinearColorRGB(Color=<rgba 10 3.6589 0 0.1>), B=<1 1 1>)
     "Float (Alpha)"=&Multiply_Float(A=0.1, B=0.8)
    )
    Create Color Attribute=true
    Create Scale Attribute=true
    Scale=&Multiply_VectorByFloat(Vector=&Multiply_Vector(A=<1.3 1.3 1>, B=$System.SizeAll), Float=1.5)
    Write to Module Namespace=true
   ) {module=FN_Mesh_Attributes002}
   &FN_Mesh_Attributes(
    Color=&MakeLinearColorFromVectorAndFloat(
     "Vector (RGB)"=&Multiply_Vector(A=&MakeVectorFromLinearColorRGB(Color=<rgba 1 0.7244 0.3072 1>), B=<1 1 1>)
     "Float (Alpha)"=&Multiply_Float(A=1, B=1)
    )
    Create Color Attribute=true
    Create Scale Attribute=true
    Scale=&Multiply_VectorByFloat(Vector=&Multiply_Vector(A=<1 1 1>, B=$System.SizeAll), Float=1.5)
    Write to Module Namespace=true
   ) {module=FN_Mesh_Attributes003}
   /SetVariables(Particles.Dy2:Vector4f=&MakeVector4(X=0, Y=0.1, Z=0, W=0)) {isSetParameters=true}
   /SetVariables(Particles.Dy3:Vector4f=&MakeVector4(X=0, Y=0.01, Z=0, W=0)) {isSetParameters=true}
   /SetVariables(Particles.Dy4:Vector4f=&MakeVector4(X=0, Y=0.5, Z=0, W=0)) {isSetParameters=true}
   /SetVariables(Particles.Dy5:Vector4f=&MakeVector4(X=0, Y=1, Z=0, W=0)) {isSetParameters=true}
   /SetVariables(Particles.Position2:NiagaraPosition=&AddVectorToPosition(Position=$Particles.Position, Vector=<0 -20 0>)) {isSetParameters=true}
   /SetVariables(Particles.Position3:NiagaraPosition=&AddVectorToPosition(Position=$Particles.Position, Vector=<0 -50 0>)) {isSetParameters=true}
   /SetVariables(Particles.Position4:NiagaraPosition=&AddVectorToPosition(Position=$Particles.Position, Vector=<0 0 0>)) {isSetParameters=true}
   /SetVariables(Particles.Position5:NiagaraPosition=&AddVectorToPosition(Position=$Particles.Position, Vector=<0 -50 0>)) {isSetParameters=true}
   &CameraOffset(Camera Offset Amount=$System.CameraOffsetAll)
  ]
  rendererClasses [&NiagaraMeshRendererProperties]
  renderers [
   {rendererClass=&NiagaraMeshRendererProperties, rendererIndex=0}
   {rendererClass=&NiagaraMeshRendererProperties, rendererIndex=1}
   {rendererClass=&NiagaraMeshRendererProperties, rendererIndex=2}
   {rendererClass=&NiagaraMeshRendererProperties, rendererIndex=3}
   {rendererClass=&NiagaraMeshRendererProperties, rendererIndex=4}
  ]
 }
 MeshTrail {
  emitterUpdateScript [
   &EmitterState(Life Cycle Mode=::NewEnumerator1, Loop Behavior=::NewEnumerator1, Loop Duration=0.6, UseLoopDelay=true)
   &SpawnBurst_Instantaneous(Use Spawn Probability=false, UseLoopCountLimit=false)
  ]
  particleSpawnScript [
   &PartitionParticles()
   &InitializeParticle(
    Color=<rgba 1 0.95 0.5458 1>
    Lifetime=$System.LifetimeAll
    Mesh Scale=&Multiply_VectorByFloat(Vector=<1 1 1>, Float=1.3)
    Mesh Scale Mode=::NewEnumerator4
    Position Offset=<0 -100 0>
    UsePositionOffset=true
    Write Material Random=true
   )
   &InitialMeshOrientation(Mesh Orientation Mode=::NewEnumerator7, Rotation=<0 0 0.5>, Use Rotation Vector=true)
   &InitialMeshOrientation(Mesh Orientation Mode=::NewEnumerator7, Rotation=<0 0 -0.25>, Use Rotation Vector=true) {module=InitialMeshOrientation001}
   &InitialMeshOrientation(
    Mesh Orientation Mode=::NewEnumerator2
    Rotation=&SelectVectorFromArray(
     Direct Array Index=$Particles.PartitionParticles.Partition
     Vector Selection Array=DataInterface() {propertyValues {FloatData [<0 0 0>, <0 0.25 0>]}}
    )
    Use Rotation Vector=true
   ) {module=InitialMeshOrientation002}
   &AddVelocity(Velocity=$System.VelocityAll)
  ]
  particleUpdateScript [
   &Drag(Drag=$System.DragAll, Use Linear Drag=true, Use Rotational Drag=true) {module=Drag001}
   &SolveForcesAndVelocity(Clamp Velocity=false, Limit Acceleration=false)
   &ParticleState()
   &ScaleMeshSize(Scale Factor=&VectorFromCurve(
     VectorCurve=DataInterface() {propertyValues {XCurve {keys [<at 0 1>,<at 1 1>]}, YCurve {keys [<at 0 1>,<at 1 1>]}, ZCurve {keys [<at 0 1>,<at 1 1>]}}}
     Scale Curve=<1.2 1 1>
    )) {module=ScaleMeshSize001}
   &ScaleMeshSize(Scale Factor=&Multiply_VectorByFloat(Vector=<1 1 1>, Float=$System.SizeScale02))
   &ScaleColor(Scale RGB=&MakeVectorFromLinearColorRGB(Color=<rgba 0.5 0.3791 0.1517 1>), ScaleA=true, ScaleRGB=true)
   &DynamicMaterialParameters(
    Index 0 Param 1=2
    Index 0 Param 2=100
    Module.Index0Param0Write=true
    Module.Index0Param1Write=true
    Module.Index0Param2Write=true
    Module.Index0Param3Write=true
   )
   &FN_Mesh_Attributes(
    Color=&MakeLinearColorFromVectorAndFloat("Vector (RGB)"=&MakeVectorFromLinearColorRGB(Color=<rgba 1 0.611 0.222 5>), "Float (Alpha)"=0.5)
    Create Color Attribute=true
    Create Position Attribute=true
    Create Scale Attribute=true
    Scale=&Multiply_Vector(A=<1.2 1 1>, B=&Multiply_VectorByFloat(Vector=<1 1 1>, Float=8))
    Write to Module Namespace=true
   )
   &CameraOffset(Camera Offset Amount=$System.CameraOffsetAll)
  ]
  rendererClasses [&NiagaraMeshRendererProperties]
  renderers [{rendererClass=&NiagaraMeshRendererProperties, rendererIndex=0}]
 }
 Sparks {
  emitterUpdateScript [
   &EmitterState(Life Cycle Mode=::NewEnumerator1, Loop Behavior=::NewEnumerator1, Loop Duration=0.6, UseLoopDelay=false)
   &SpawnParticlesFromOtherEmitter(
    Attribute Reader=DataInterface() {propertyValues {EmitterBinding {bindingMode=Other, emitterName=HeadMesh}}}
    Spawn Rate=300
    Spawn Rate Per Particle Cap=300
    Use Maximum Age=false
    Use Minimum Age=false
   )
  ]
  particleSpawnScript [
   &InitializeParticle(
    Color=<rgba 10 8.855 0 1>
    Lifetime Max=0.4
    Lifetime Min=0.2
    Lifetime Mode=::NewEnumerator1
    Sprite Size Max=<0.3 2>
    Sprite Size Min=<0.2 2>
    Sprite Size Mode=::NewEnumerator2
    UsePositionOffset=false
    Write Material Random=true
   )
   &SampleParticlesFromOtherEmitter(Applied Velocity Scale=0.1, Velocity Sampling=::NewEnumerator0)
   &ShapeLocation(
    Box Size=<150 0 1>
    MaskSpawnGroup=false
    Offset=<0 -100 0>
    Offset Mode=::NewEnumerator0
    Override Seed=false
    Shape Primitive=::NewEnumerator2
    Surface Only=false
    UseOwnerScale=true
   )
   &AddVelocity(
    Cone Axis=<0 1 0>
    Module.UseConeFalloff=false
    Module.UseVelDistribution=false
    Override Seed=false
    Velocity Mode=::NewEnumerator2
    Velocity Speed=&RandomRangeFloat(Minimum=500, Maximum=700)
   ) {module=AddVelocity001}
  ]
  particleUpdateScript [
   &ParticleState()
   &ScaleSpriteSize(Scale Sprite Size Mode=::NewEnumerator0, Uniform Scale Factor=2)
   &CurlNoiseForce(Bias Noise=false, Noise Frequency=30, Noise Strength=1000, Pan Noise=false)
   &Drag(Drag=3, Use Linear Drag=true, Use Rotational Drag=true)
   &SolveForcesAndVelocity(Clamp Velocity=false, Limit Acceleration=false)
   &FN_Sprite_Attributes(
    Color=<rgba 1 0.7033 0.1099 0.1>
    Create Color Attribute=true
    Create SpriteSize Attribute=true
    SpriteSize=&Multiply_Vector2D(A=$Particles.SpriteSize, B=<20 4>)
    Write to Module Namespace=true
   )
   &DynamicMaterialParameters(
    Index 0 Param 1=2
    Module.Index0Param0Write=true
    Module.Index0Param1Write=true
    Module.Index0Param2Write=true
    Module.Index0Param3Write=true
   )
  ]
  rendererClasses [&NiagaraSpriteRendererProperties]
  renderers [
   {rendererClass=&NiagaraSpriteRendererProperties}
   {rendererClass=&NiagaraSpriteRendererProperties, rendererIndex=1}
  ]
 }
}
summary {
 systemName=NS_SinisterMelon_ComboEndAOE
 emitters [
  {emitterName=HeadMesh, rendererClasses [&NiagaraMeshRendererProperties]}
  {emitterName=MeshTrail, rendererClasses [&NiagaraMeshRendererProperties]}
  {emitterName=Glow, rendererClasses [&NiagaraSpriteRendererProperties]}
  {emitterName=Sparks, rendererClasses [&NiagaraSpriteRendererProperties]}
 ]
 userVariables [
  User.bIsFrontEnd {type=NiagaraBool}
  User.bIsFrontEndPreview {type=NiagaraBool}
  User.Physical Surface Dust {type=NiagaraBool}
  User.VariantSwitch {type=NiagaraFloat, defaultValue=0, description=""}
 ]
}
