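# Niagara system for Sinister Melon's combo slash effect: a short, warm-coloured swirling slash with sparks.
# Slash and Energy each burst mesh particles that follow owner rotation and scale, spin with decelerating curves, and fade.
# Slash has three mesh renderers, using animated dynamic material parameters and extra module-namespaced colour/scale attributes to layer looks on one particle.
# LeakParticle adds a delayed GPU burst of stretched sprites driven by vortex force and drag.
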NiagaraSystem /SinisterMelon/Gameplay/SinisterMelon/FX/NewFX/NS_SinisterMelon_ComboSlash
dependencies {
 usedDataInterfaces [&NiagaraDataInterfaceCurve, &NiagaraDataInterfaceColorCurve]
 usedDynamicInputs []
 usedModules [
  &self:SystemSpawnScript.SystemScriptSource.SystemScriptGraph.NiagaraNodeAssignment_0.SetVariables_#1
  &self:SystemSpawnScript.SystemScriptSource.SystemScriptGraph.NiagaraNodeAssignment_1.SetVariables_#2
  &SystemState
 ]
}
emitters {
 Energy {
  emitterUpdateScript [
   &EmitterState(Life Cycle Mode=::NewEnumerator1, Loop Behavior=::NewEnumerator1, Loop Duration=1, UseLoopDelay=false)
   &SpawnBurst_Instantaneous(Use Spawn Probability=false, UseLoopCountLimit=false)
  ]
  particleSpawnScript [
   &InitializeParticle(
    Color=<rgba 3 1.6372 0.1204 7>
    Lifetime=0.37
    Mesh Scale=&Multiply_VectorByFloat(Vector=<1 1 0.3>, Float=1.9)
    Mesh Scale Mode=::NewEnumerator4
    Position Offset=<0 0 5>
    UsePositionOffset=true
    Write Material Random=true
   )
   /SetVariables(Particles.MeshOrientation=$Engine.Owner.Rotation) {isSetParameters=true, module=SetVariables_#3}
   &CameraOffset()
  ]
  particleUpdateScript [
   &ParticleState()
   &ScaleMeshSize(Scale Factor=&Multiply_Vector(
     A=$Engine.Owner.Scale
     B=&Multiply_Vector(
      A=$System.SizeScale
      B=&MakeVector(
       X=1
       Y=1
       Z=&FloatFromCurve(FloatCurve=DataInterface() {propertyValues {Curve {keys [<at 0 0.5>,<at 1 1>]}}}, Scale Curve=2)
      )
     )
    ))
   &UpdateMeshOrientation(
    Rotation Rate=&FloatFromCurve(
     FloatCurve=DataInterface() {propertyValues {Curve {keys [{arriveTangent=-3.0754, interpMode=RCIM_Cubic, leaveTangent=-3.0754, tangentMode=RCTM_User, time=0, value=2}, <at 0.5 0.2>]}}}
     Scale Curve=2.2
    )
    Rotation Vector=<0 0 -1.3>
   ) {module=UpdateMeshOrientation001}
   &ScaleColor(
    Scale Alpha=&FloatFromCurve(
     FloatCurve=DataInterface() {propertyValues {Curve {keys [
         {interpMode=RCIM_Cubic, tangentMode=RCTM_User, time=0, value=1}
         {arriveTangent=-0.6572, interpMode=RCIM_Cubic, leaveTangent=-0.6572, tangentMode=RCTM_User, time=1, value=0}
        ]}}}
     Scale Curve=0.25
    )
    Scale RGB=&Multiply_VectorByFloat(
     Vector=&MakeVectorFromLinearColorRGB(Color=&ColorFromCurve(ColorCurve=DataInterface() {propertyValues {
         AlphaCurve {keys []}
         BlueCurve {keys [<at 0 0.0729 RCIM_Cubic>]}
         GreenCurve {keys [<at 0 1 RCIM_Cubic>]}
         RedCurve {keys [<at 0 1 RCIM_Cubic>]}
        }}))
     Float=1
    )
    ScaleA=true
    ScaleRGB=true
   )
   /SetVariables() {isSetParameters=true, module=SetVariables_#4}
  ]
  rendererClasses [&NiagaraMeshRendererProperties]
  renderers [{rendererClass=&NiagaraMeshRendererProperties, rendererIndex=0}]
 }
 LeakParticle {
  simTarget=GPUComputeSim
  emitterUpdateScript [
   &EmitterState(Life Cycle Mode=::NewEnumerator1, Loop Behavior=::NewEnumerator1, Loop Duration=0.4, UseLoopDelay=true)
   &SpawnBurst_Instantaneous(Spawn Count=20, Use Spawn Probability=false, UseLoopCountLimit=false)
  ]
  particleSpawnScript [
   &PartitionParticles()
   &InitializeParticle(
    Color=<rgba 2 0.3757 0 1>
    Lifetime Max=0.4
    Lifetime Min=0.2
    Lifetime Mode=::NewEnumerator1
    Sprite Size Max=<0.5 6>
    Sprite Size Min=<0.3 2>
    Sprite Size Mode=::NewEnumerator2
    UsePositionOffset=false
    Write Material Random=true
   )
   &CameraOffset()
   &ShapeLocation(
    Handle Radius=1
    MaskSpawnGroup=false
    Override Seed=false
    Rotation Angle=0.3
    Rotation Axis=<0 0 1>
    Rotation Mode=::NewEnumerator1
    Shape Primitive=::NewEnumerator3
    Surface Distribution=1
    U Distribution=0.7
    UseOwnerScale=true
    V Distribution=1
   )
   &InitialMeshOrientation(
    Mesh Orientation Mode=::NewEnumerator6
    Orientation Quaternion:Quat4f=&FindQuatBetweenTwoVectors(
     From Vector=<0 0 1>
     To Vector=&RotateVectorByQuaternion(Vector To Rotate=<0 1 0>, Quaternion=$Particles.SkeletalMeshLocation001.SampledRotation)
    )
    Use Rotation Vector=false
   )
   &AddVelocity(
    Module.Use Falloff Distance=false
    Override Seed=false
    Velocity Mode=::NewEnumerator1
    Velocity Speed=&RandomRangeFloat(Minimum=50, Maximum=200)
   )
  ]
  particleUpdateScript [
   &ParticleState()
   &ScaleSpriteSize(
    Scale Factor=&MakeVector2D(
     X=&FloatFromCurve(FloatCurve=DataInterface() {propertyValues {Curve {keys [<at 0 0.2>,<at 1 1>]}}})
     Y=1
    )
    Scale Sprite Size Mode=::NewEnumerator2
   ) {module=ScaleSpriteSize002}
   &ScaleSpriteSize(
    Scale Factor=&MakeVector2D(
     X=1
     Y=&FloatFromCurve(FloatCurve=DataInterface() {propertyValues {Curve {keys [{interpMode=RCIM_Cubic, leaveTangent=-10.3145, tangentMode=RCTM_Break, time=0, value=4}, <at 1 1>]}}})
    )
    Scale Sprite Size Mode=::NewEnumerator2
   ) {module=ScaleSpriteSize001}
   &ScaleSpriteSize(Scale Sprite Size Mode=::NewEnumerator0, Uniform Scale Factor=2)
   &CurlNoiseForce(Bias Noise=false, Noise Frequency=60, Noise Strength=200, Pan Noise=true, Pan Noise Field=<0 0 1>) {enabled=false}
   &Drag(Drag=3, Use Linear Drag=true, Use Rotational Drag=true)
   &GravityForce(Gravity=<0 0 0>) {enabled=false}
   &VortexForce(UseOriginOffset=false, Vortex Force Amount=&RandomRangeFloat(Minimum=-1000, Maximum=-1000))
   &ScaleVelocity() {enabled=false}
   /SetVariables(Transient.PhysicsForce:Vector3f=$Transient.PhysicsForce) {enabled=false, isSetParameters=true}
   &SolveForcesAndVelocity(Clamp Velocity=false, Limit Acceleration=false)
   &ScaleSpriteSizeBySpeed(Max Scale Factor=<1 1>, Velocity Threshold=5633.8472)
   &ScaleColor(
    Scale Alpha=&FloatFromCurve(FloatCurve=DataInterface() {propertyValues {Curve {keys [
         {tangentMode=RCTM_User, time=0, value=1}
         {arriveTangent=-3.3056, interpMode=RCIM_Cubic, leaveTangent=-3.3056, tangentMode=RCTM_User, time=0.2, value=1}
         {arriveTangent=-0.0264, interpMode=RCIM_Cubic, leaveTangent=-0.0264, tangentMode=RCTM_User, time=1, value=0}
        ]}}})
    Scale RGB=&VectorFromFloat(Value=10)
    ScaleA=true
    ScaleRGB=true
   )
   &FN_Sprite_Attributes(
    Color=&MultiplyLinearColor(
     Linear Color=<rgba 0.151 0.0606 0 1>
     Scale Factor=&MakeVector4(
      X=1
      Y=1
      Z=1
      W=&FloatFromCurve(
       FloatCurve=DataInterface() {propertyValues {Curve {keys [
           {arriveTangent=-2.2465, interpMode=RCIM_Cubic, leaveTangent=-2.2465, tangentMode=RCTM_User, time=0, value=1}
           {arriveTangent=-0.4533, interpMode=RCIM_Cubic, leaveTangent=-0.4533, tangentMode=RCTM_User, time=1, value=0}
          ]}}}
       Scale Curve=0.5816
      )
     )
    )
    Create Color Attribute=true
    Create SpriteSize Attribute=true
    SpriteSize=<100 150>
    Write to Module Namespace=true
   )
  ]
  rendererClasses [&NiagaraSpriteRendererProperties]
  renderers [
   {rendererClass=&NiagaraSpriteRendererProperties}
   {rendererClass=&NiagaraSpriteRendererProperties, rendererIndex=1}
  ]
 }
 Slash {
  emitterUpdateScript [
   &EmitterState(Life Cycle Mode=::NewEnumerator1, Loop Behavior=::NewEnumerator1, Loop Duration=1, UseLoopDelay=false)
   &SpawnBurst_Instantaneous(Use Spawn Probability=false, UseLoopCountLimit=false)
  ]
  particleSpawnScript [
   &InitializeParticle(
    Lifetime=0.4
    Mesh Scale=&Multiply_VectorByFloat(Vector=<1 1 0.3>, Float=2)
    Mesh Scale Mode=::NewEnumerator4
    UsePositionOffset=false
    Write Material Random=true
   )
   /SetVariables(Particles.MeshOrientation=$Engine.Owner.Rotation) {isSetParameters=true, module=SetVariables_#3}
   &CameraOffset()
  ]
  particleUpdateScript [
   &ParticleState()
   &ScaleMeshSize(Scale Factor=&Multiply_Vector(
     A=$Engine.Owner.Scale
     B=&Multiply_Vector(
      A=$System.SizeScale
      B=&MakeVector(
       X=1
       Y=1
       Z=&FloatFromCurve(FloatCurve=DataInterface() {propertyValues {Curve {keys [<at 0 0.5>,<at 1 1>]}}})
      )
     )
    ))
   &UpdateMeshOrientation(
    Rotation Rate=&FloatFromCurve(
     FloatCurve=DataInterface() {propertyValues {Curve {keys [{arriveTangent=-3.0754, interpMode=RCIM_Cubic, leaveTangent=-3.0754, tangentMode=RCTM_User, time=0, value=2}, <at 0.5 0.1>]}}}
     Scale Curve=2.5
    )
    Rotation Vector=<0 0 -1>
   ) {module=UpdateMeshOrientation001}
   &ScaleColor(
    Scale RGB=&Multiply_VectorByFloat(
     Vector=&MakeVectorFromLinearColorRGB(Color=&ColorFromCurve(ColorCurve=DataInterface() {propertyValues {
         AlphaCurve {keys []}
         BlueCurve {keys [<at 0 0.5889 RCIM_Cubic>,<at 0.8853 0.5677 RCIM_Cubic>]}
         GreenCurve {keys [<at 0 0.6963 RCIM_Cubic>,<at 0.8853 0.7983 RCIM_Cubic>]}
         RedCurve {keys [<at 0 1 RCIM_Cubic>,<at 0.8853 1 RCIM_Cubic>]}
        }}))
     Float=&FloatFromCurve(FloatCurve=DataInterface() {propertyValues {Curve {keys [
          <at 0.1 1 RCIM_Cubic>
          {arriveTangent=-1.5658, interpMode=RCIM_Cubic, leaveTangent=-1.5658, tangentMode=RCTM_User, time=0.2911, value=0.9235}
          {arriveTangent=-2.0316, interpMode=RCIM_Cubic, leaveTangent=-2.0316, tangentMode=RCTM_User, time=1, value=0.1}
         ]}}})
    )
    ScaleA=true
    ScaleRGB=true
   )
   &DynamicMaterialParameters(
    Index 0 Param 1=&FloatFromCurve(FloatCurve=DataInterface() {propertyValues {Curve {keys [<at 0 0 RCIM_Cubic>,<at 1 1>]}}})
    Index 0 Param 2=0.15
    Index 0 Param 3=&FloatFromCurve(FloatCurve=DataInterface() {propertyValues {Curve {
        defaultValue=0
        keys [
         {arriveTangent=1.2925, interpMode=RCIM_Cubic, leaveTangent=1.2925, tangentMode=RCTM_User, time=0, value=0.2}
         {arriveTangent=0.6971, interpMode=RCIM_Cubic, leaveTangent=0.6971, tangentMode=RCTM_User, time=1, value=0.9}
        ]
       }}})
    Index 1 Param 1=&FloatFromCurve(FloatCurve=DataInterface() {propertyValues {Curve {keys [<at 0 0 RCIM_Cubic>,<at 1 1>]}}})
    Index 1 Param 2=&FloatFromCurve(FloatCurve=DataInterface() {propertyValues {Curve {keys [<at 0 1>,<at 0.3 0.05>,<at 1 0>]}}})
    Index 1 Param 3=&FloatFromCurve(FloatCurve=DataInterface() {propertyValues {Curve {
        defaultValue=0
        keys [
         {arriveTangent=1.2925, interpMode=RCIM_Cubic, leaveTangent=1.2925, tangentMode=RCTM_User, time=0, value=0.2}
         {arriveTangent=0.6971, interpMode=RCIM_Cubic, leaveTangent=0.6971, tangentMode=RCTM_User, time=1, value=0.9}
        ]
       }}})
    Index 2 Param 1=&FloatFromCurve(FloatCurve=DataInterface() {propertyValues {Curve {keys [<at 0 0 RCIM_Cubic>,<at 1 1>]}}})
    Index 2 Param 2=1
    Index 2 Param 3=&FloatFromCurve(FloatCurve=DataInterface() {propertyValues {Curve {
        defaultValue=0
        keys [
         {arriveTangent=1.2925, interpMode=RCIM_Cubic, leaveTangent=1.2925, tangentMode=RCTM_User, time=0, value=0.2}
         {arriveTangent=0.6971, interpMode=RCIM_Cubic, leaveTangent=0.6971, tangentMode=RCTM_User, time=1, value=0.9}
        ]
       }}})
    Module.Index0Param0Write=true
    Module.Index0Param1Write=true
    Module.Index0Param2Write=true
    Module.Index0Param3Write=true
    Module.Index1Param0Write=true
    Module.Index1Param1Write=true
    Module.Index1Param2Write=true
    Module.Index1Param3Write=true
    Module.Index2Param0Write=true
    Module.Index2Param1Write=true
    Module.Index2Param2Write=true
    Module.Index2Param3Write=true
    Write Parameter Index 1=true
    Write Parameter Index 2=true
   )
   /SetVariables() {isSetParameters=true, module=SetVariables_#4}
   &FN_Mesh_Attributes(
    Color=&MakeLinearColorFromVectorAndFloat(
     "Vector (RGB)"=&Multiply_Vector(
      A=&Multiply_Vector(
       A=&MakeVectorFromLinearColorRGB(Color=<rgba 0.9792 0.2808 0.1411 1>)
       B=&Multiply_VectorByFloat(
        Vector=&MakeVectorFromLinearColorRGB(Color=&ColorFromCurve(ColorCurve=DataInterface() {propertyValues {
            AlphaCurve {keys []}
            BlueCurve {keys [<at 0 0.5889 RCIM_Cubic>]}
            GreenCurve {keys [<at 0 0.6963 RCIM_Cubic>]}
            RedCurve {keys [<at 0 1 RCIM_Cubic>]}
           }}))
        Float=&FloatFromCurve(FloatCurve=DataInterface() {propertyValues {Curve {keys [
             <at 0.1 1 RCIM_Cubic>
             {arriveTangent=-1.5658, interpMode=RCIM_Cubic, leaveTangent=-1.5658, tangentMode=RCTM_User, time=0.3235, value=1}
             <at 0.6167 1.3889 RCIM_Cubic>
             {arriveTangent=-2.0316, interpMode=RCIM_Cubic, leaveTangent=-2.0316, tangentMode=RCTM_User, time=1, value=1}
            ]}}})
       )
      )
      B=&MakeVectorFromLinearColorRGB(Color=&ColorFromCurve(ColorCurve=DataInterface() {propertyValues {
          AlphaCurve {keys []}
          BlueCurve {keys [<at 0 3.64 RCIM_Cubic>,<at 0.2432 1 RCIM_Cubic>]}
          GreenCurve {keys [<at 0 10 RCIM_Cubic>,<at 0.2432 1 RCIM_Cubic>]}
          RedCurve {keys [<at 0 8.5627 RCIM_Cubic>,<at 0.2432 1 RCIM_Cubic>]}
         }}))
     )
     "Float (Alpha)"=&FloatFromCurve(
      FloatCurve=DataInterface() {propertyValues {Curve {keys [<at 0 1 RCIM_Cubic>, {arriveTangent=-0.5666, tangentMode=RCTM_User, time=1, value=0.1}]}}}
      Scale Curve=2
     )
    )
    Create Color Attribute=true
    Create Scale Attribute=true
    Scale=&Multiply_Vector(
     A=&Multiply_VectorByFloat(Vector=<1 1 0.3>, Float=2)
     B=&Multiply_Vector(
      A=$Engine.Owner.Scale
      B=&Multiply_Vector(
       A=$System.SizeScale
       B=&MakeVector(
        X=1
        Y=1
        Z=&FloatFromCurve(FloatCurve=DataInterface() {propertyValues {Curve {keys [<at 0 0.5>,<at 1 1>]}}})
       )
      )
     )
    )
    Write to Module Namespace=true
   )
   &FN_Mesh_Attributes(
    Color=&MakeLinearColorFromVectorAndFloat(
     "Vector (RGB)"=&Multiply_Vector(
      A=&Multiply_Vector(
       A=&MakeVectorFromLinearColorRGB(Color=<rgba 0.4688 0.4057 0.0905 1>)
       B=&Multiply_VectorByFloat(
        Vector=&MakeVectorFromLinearColorRGB(Color=&ColorFromCurve(ColorCurve=DataInterface() {propertyValues {
            AlphaCurve {keys []}
            BlueCurve {keys [<at 0 0.5889 RCIM_Cubic>,<at 0.8853 0.2916 RCIM_Cubic>]}
            GreenCurve {keys [<at 0 0.6963 RCIM_Cubic>,<at 0.8853 0.8929 RCIM_Cubic>]}
            RedCurve {keys [<at 0 1 RCIM_Cubic>,<at 0.8853 1 RCIM_Cubic>]}
           }}))
        Float=&FloatFromCurve(FloatCurve=DataInterface() {propertyValues {Curve {keys [
             <at 0.1 1 RCIM_Cubic>
             {arriveTangent=0.6683, interpMode=RCIM_Cubic, leaveTangent=0.6683, tangentMode=RCTM_User, time=0.2733, value=0.467}
             {arriveTangent=-0.5419, interpMode=RCIM_Cubic, leaveTangent=-0.5419, tangentMode=RCTM_User, time=1, value=1}
            ]}}})
       )
      )
      B=<1 1 1>
     )
     "Float (Alpha)"=&FloatFromCurve(FloatCurve=DataInterface() {propertyValues {Curve {keys [<at 0 1>,<at 1 0>]}}}, Scale Curve=1.8)
    )
    Create Color Attribute=true
    Create Scale Attribute=true
    Scale=&Multiply_Vector(
     A=&Multiply_VectorByFloat(Vector=<1 1 0.3>, Float=2)
     B=&Multiply_Vector(
      A=$Engine.Owner.Scale
      B=&Multiply_Vector(
       A=$System.SizeScale
       B=&MakeVector(
        X=1
        Y=1
        Z=&FloatFromCurve(FloatCurve=DataInterface() {propertyValues {Curve {keys [<at 0 0.5>,<at 1 1>]}}})
       )
      )
     )
    )
    Write to Module Namespace=true
   ) {module=FN_Mesh_Attributes001}
  ]
  rendererClasses [&NiagaraMeshRendererProperties]
  renderers [
   {rendererClass=&NiagaraMeshRendererProperties, rendererIndex=0}
   {rendererClass=&NiagaraMeshRendererProperties, rendererIndex=1}
   {rendererClass=&NiagaraMeshRendererProperties, rendererIndex=2}
  ]
 }
}
summary {
 systemName=NS_SinisterMelon_ComboSlash
 emitters [
  {emitterName=LeakParticle, simTarget=GPUComputeSim, rendererClasses [&NiagaraSpriteRendererProperties]}
  {emitterName=Slash, rendererClasses [&NiagaraMeshRendererProperties]}
  {emitterName=Energy, rendererClasses [&NiagaraMeshRendererProperties]}
 ]
 userVariables []
}
