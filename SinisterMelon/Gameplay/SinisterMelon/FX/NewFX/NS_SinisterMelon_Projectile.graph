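# Niagara effect for the Sinister Melon projectile, built around a two-piece cross mesh with layered renderers.
# Airflow, trail and spiral meshes, glow and flash sprites, directional smoke and curl-noise GPU sparks surround it.
# Four system collide flags (left, up, right, down) drive cross material parameters, stop smoke velocity per direction, and hide layers or suppress sparks when certain combinations are set.
# System size variables scale the main layers, while distance remaps adjust some sprites.
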NiagaraSystem /SinisterMelon/Gameplay/SinisterMelon/FX/NewFX/NS_SinisterMelon_Projectile
dependencies {
 usedDataInterfaces [
  &NiagaraDataInterfaceArrayFloat3
  &NiagaraDataInterfaceColorCurve
  &NiagaraDataInterfaceVectorCurve
  &NiagaraDataInterfaceCurve
  &NiagaraDataInterfaceVector2DCurve
  &NiagaraDataInterfaceSpriteRendererInfo
 ]
 usedDynamicInputs [&Vector2DFromCurve, &RandomRangeVector2D]
 usedModules [
  &SystemState
  &self:SystemSpawnScript.SystemScriptSource.SystemScriptGraph.NiagaraNodeAssignment_0.SetVariables_#1
  &self:SystemSpawnScript.SystemScriptSource.SystemScriptGraph.NiagaraNodeAssignment_2.SetVariables_#2
  &self:CheckHidingState
 ]
}
emitters {
 Airflow {
  emitterUpdateScript [
   &EmitterState(Life Cycle Mode=::NewEnumerator1, Loop Behavior=::NewEnumerator1, Loop Duration Mode=::NewEnumerator1, UseLoopDelay=false)
   &SpawnBurst_Instantaneous(Spawn Count=2, Use Spawn Probability=false, UseLoopCountLimit=false)
  ]
  particleSpawnScript [
   &PartitionParticles()
   &InitializeParticle(
    Color=<rgba 3 1.7923 0 1>
    Mesh Scale Mode=::NewEnumerator3
    Mesh Uniform Scale=1.6
    Position Offset=<0 -110 0>
    UsePositionOffset=true
    Write Material Random=true
   )
   &InitialMeshOrientation(
    Mesh Orientation Mode=::NewEnumerator7
    Rotation=&SelectVectorFromArray(
     Direct Array Index=$Particles.PartitionParticles.Partition
     Vector Selection Array=DataInterface() {propertyValues {FloatData [<0 0 0>, <0 0.25 0>]}}
    )
    Use Rotation Vector=true
   )
   &InitialMeshOrientation(Mesh Orientation Mode=::NewEnumerator2, Rotation=<0 0.5 0>, Use Rotation Vector=true) {module=InitialMeshOrientation001}
  ]
  particleUpdateScript [
   &ParticleState(Kill Particles When Lifetime Has Elapsed=false)
   &ScaleMeshSize(Scale Factor=&Multiply_VectorByFloat(Vector=<1 0.7 1.25>, Float=$System.SizeScale02))
   &DynamicMaterialParameters(
    Index 0 Param 2=0.01
    Module.Index0Float0Write=true
    Module.Index0Float1Write=true
    Module.Index0Float2Write=true
    Module.Index0Float3Write=true
   )
   &ScaleColor(
    Scale Alpha=&MakeCustomFloatFromBool(
     Bool=&Boolean_OrOperation(
      Bool1=&Boolean_AndOperation(Bool1=$System.IsLeftCollide, Bool2=$System.IsUpCollide)
      Bool2=&Boolean_AndOperation(Bool1=$System.IsRightCollide, Bool2=$System.IsDownCollide)
     )
     True Float=0
     False Float=1
    )
    ScaleA=true
    ScaleRGB=true
   )
   &FN_Mesh_Attributes(
    Color=&MakeLinearColorFromVectorAndFloat(
     "Vector (RGB)"=&Multiply_Vector(A=&MakeVectorFromLinearColorRGB(Color=<rgba 0.0677 0.038 0 1>), B=<1 1 1>)
     "Float (Alpha)"=&Multiply_Float(
      A=&MakeCustomFloatFromBool(
       Bool=&Boolean_OrOperation(
        Bool1=&Boolean_AndOperation(Bool1=$System.IsLeftCollide, Bool2=$System.IsUpCollide)
        Bool2=&Boolean_AndOperation(Bool1=$System.IsRightCollide, Bool2=$System.IsDownCollide)
       )
       True Float=0
       False Float=1
      )
      B=1
     )
    )
    Create Color Attribute=true
    Create Scale Attribute=true
    Scale=&Multiply_Vector(
     A=<1 0.7 1.25>
     B=&Multiply_VectorByFloat(Vector=&Multiply_VectorByFloat(Vector=<1 1 1>, Float=$System.SizeScale02), Float=1.6)
    )
    Write to Module Namespace=true
   )
  ]
  rendererClasses [&NiagaraMeshRendererProperties]
  renderers [
   {rendererClass=&NiagaraMeshRendererProperties, rendererIndex=0}
   {rendererClass=&NiagaraMeshRendererProperties, rendererIndex=1}
  ]
 }
 Cross {
  emitterUpdateScript [
   &EmitterState(Life Cycle Mode=::NewEnumerator1, Loop Behavior=::NewEnumerator1, Loop Duration Mode=::NewEnumerator1, UseLoopDelay=false)
   &SpawnBurst_Instantaneous(Spawn Count=2, Use Spawn Probability=false, UseLoopCountLimit=false)
  ]
  particleSpawnScript [
   &PartitionParticles()
   &InitializeParticle(
    Color=<rgba 0.5 0.3249 0.0078 0.7>
    Mesh Scale Mode=::NewEnumerator3
    Mesh Uniform Scale=1.5
    Position Offset=<0 50 0>
    UsePositionOffset=true
    Write Material Random=true
   )
   &InitialMeshOrientation(
    Mesh Orientation Mode=::NewEnumerator2
    Rotation=&SelectVectorFromArray(
     Direct Array Index=$Particles.PartitionParticles.Partition
     Vector Selection Array=DataInterface() {propertyValues {FloatData [<0 0 0>, <0 0.25 0>]}}
    )
    Use Rotation Vector=true
   )
  ]
  particleUpdateScript [
   &ParticleState(Kill Particles When Lifetime Has Elapsed=false)
   &ScaleMeshSize(Scale Factor=<1.3 1.3 1>) {module=ScaleMeshSize001}
   &ScaleMeshSize(Scale Factor=$System.SizeAll)
   &DynamicMaterialParameters(
    Index 0 Param 2=0.2
    Index 3 Param 1=&MergeHelper_FloatSwitch(
     0=&MakeCustomFloatFromBool(Bool=$System.IsLeftCollide, True Float=1, False Float=0)
     1=&MakeCustomFloatFromBool(Bool=$System.IsUpCollide, True Float=1, False Float=0)
     2=?
     3=?
     Index=$Particles.PartitionParticles.Partition
     Num=::NewEnumerator0
    )
    Index 3 Param 2=&MergeHelper_FloatSwitch(
     0=&MakeCustomFloatFromBool(Bool=$System.IsRightCollide, True Float=1, False Float=0)
     1=&MakeCustomFloatFromBool(Bool=$System.IsDownCollide, True Float=1, False Float=0)
     2=?
     3=?
     Index=$Particles.PartitionParticles.Partition
     Num=::NewEnumerator0
    )
    Module.Index0Float0Write=true
    Module.Index0Float1Write=true
    Module.Index0Float2Write=true
    Module.Index0Float3Write=true
    Module.Index3Float0Write=true
    Module.Index3Float1Write=true
    Module.Index3Float2Write=true
    Module.Index3Float3Write=true
    Write Parameter Index 3=::NewEnumerator1
   )
   &FN_Mesh_Attributes(
    Color=&MakeLinearColorFromVectorAndFloat(
     "Vector (RGB)"=&Multiply_Vector(A=&MakeVectorFromLinearColorRGB(Color=<rgba 1.5 1.2683 0.7351 2>), B=<1 1 1>)
     "Float (Alpha)"=&Multiply_Float(A=2, B=1)
    )
    Create Color Attribute=true
    Create Scale Attribute=true
    Scale=&Multiply_VectorByFloat(Vector=&Multiply_Vector(A=<1.1 1.1 0.8>, B=$System.SizeAll), Float=1.6)
    Write to Module Namespace=true
   )
   &FN_Mesh_Attributes(
    Color=&MakeLinearColorFromVectorAndFloat(
     "Vector (RGB)"=&Multiply_Vector(A=&MakeVectorFromLinearColorRGB(Color=<rgba 1.7 1.0359 0.4334 1.5>), B=<1 1 1>)
     "Float (Alpha)"=&Multiply_Float(A=1.5, B=1)
    )
    Create Color Attribute=true
    Create Scale Attribute=true
    Scale=&Multiply_VectorByFloat(Vector=&Multiply_Vector(A=<1 1 1>, B=$System.SizeAll), Float=1.5)
    Write to Module Namespace=true
   ) {module=FN_Mesh_Attributes001}
   &FN_Mesh_Attributes(
    Color=&MakeLinearColorFromVectorAndFloat(
     "Vector (RGB)"=&Multiply_Vector(A=&MakeVectorFromLinearColorRGB(Color=<rgba 10 3.6589 0 0.1>), B=<1 1 1>)
     "Float (Alpha)"=&Multiply_Float(A=0.1, B=1)
    )
    Create Color Attribute=true
    Create Scale Attribute=true
    Scale=&Multiply_VectorByFloat(Vector=&Multiply_Vector(A=<1.3 1.3 1>, B=$System.SizeAll), Float=1.5)
    Write to Module Namespace=true
   ) {module=FN_Mesh_Attributes002}
   &FN_Mesh_Attributes(
    Color=&MakeLinearColorFromVectorAndFloat(
     "Vector (RGB)"=&Multiply_Vector(A=&MakeVectorFromLinearColorRGB(Color=<rgba 1 0.7244 0.3072 1>), B=<1 1 1>)
     "Float (Alpha)"=&Multiply_Float(A=1, B=1)
    )
    Create Color Attribute=true
    Create Scale Attribute=true
    Scale=&Multiply_VectorByFloat(Vector=&Multiply_Vector(A=<1 1 1>, B=$System.SizeAll), Float=1.5)
    Write to Module Namespace=true
   ) {module=FN_Mesh_Attributes003}
   &FN_Mesh_Attributes(
    Color=&MakeLinearColorFromVectorAndFloat(
     "Vector (RGB)"=&Multiply_Vector(
      A=&MakeVectorFromLinearColorRGB(Color=<rgba 2 1.7525 0 1>)
      B=&Multiply_VectorByFloat(
       Vector=&MakeVectorFromLinearColorRGB(Color=&ColorFromCurve(ColorCurve=DataInterface() {propertyValues {
           AlphaCurve {keys []}
           BlueCurve {keys [<at 0 0.7457 RCIM_Cubic>]}
           GreenCurve {keys [<at 0 0.9788 RCIM_Cubic>]}
           RedCurve {keys [<at 0 1 RCIM_Cubic>]}
          }}))
       Float=3
      )
     )
     "Float (Alpha)"=&Multiply_Float(
      A=&MakeCustomFloatFromBool(
       Bool=&Boolean_OrOperation(
        Bool1=&Boolean_AndOperation(Bool1=$System.IsLeftCollide, Bool2=$System.IsUpCollide)
        Bool2=&Boolean_AndOperation(Bool1=$System.IsRightCollide, Bool2=$System.IsDownCollide)
       )
       True Float=0
       False Float=1
      )
      B=1
     )
    )
    Create Color Attribute=true
    Create Scale Attribute=true
    Scale=&Multiply_VectorByFloat(
     Vector=&Multiply_Vector(
      A=&Multiply_Vector(A=<1.1 0.9 0.3>, B=<1.3 1 1.2>)
      B=&VectorFromFloat(Value=$System.SizeScale02)
     )
     Float=2.2
    )
    Write to Module Namespace=true
   ) {module=FN_Mesh_Attributes004}
   /SetVariables(Particles.Dy2:Vector4f=&MakeVector4(X=0, Y=0.1, Z=0, W=0)) {isSetParameters=true}
   /SetVariables(Particles.Dy3:Vector4f=&MakeVector4(X=0, Y=0.01, Z=0, W=0)) {isSetParameters=true}
   /SetVariables(Particles.Dy4:Vector4f=&MakeVector4(X=0, Y=0.5, Z=0, W=0)) {isSetParameters=true}
   /SetVariables(Particles.Dy5:Vector4f=&MakeVector4(X=0, Y=1, Z=0, W=0)) {isSetParameters=true}
   /SetVariables(Particles.Dy6:Vector4f=&MakeVector4(X=0, Y=0.1, Z=0, W=15)) {isSetParameters=true}
   /SetVariables(Particles.Position2:NiagaraPosition=&AddVectorToPosition(Position=$Engine.Emitter.SimulationPosition, Vector=<0 20 0>)) {isSetParameters=true}
   /SetVariables(Particles.Position3:NiagaraPosition=&AddVectorToPosition(Position=$Engine.Emitter.SimulationPosition, Vector=<0 0 0>)) {isSetParameters=true}
   /SetVariables(Particles.Position4:NiagaraPosition=&AddVectorToPosition(Position=$Engine.Emitter.SimulationPosition, Vector=<0 35 0>)) {isSetParameters=true}
   /SetVariables(Particles.Position5:NiagaraPosition=&AddVectorToPosition(Position=$Engine.Emitter.SimulationPosition, Vector=<0 0 0>)) {isSetParameters=true}
   /SetVariables(Particles.Position6:NiagaraPosition=&AddVectorToPosition(Position=$Engine.Emitter.SimulationPosition, Vector=<-5 80 0>)) {isSetParameters=true}
  ]
  rendererClasses [&NiagaraMeshRendererProperties]
  renderers [
   {rendererClass=&NiagaraMeshRendererProperties, rendererIndex=0}
   {rendererClass=&NiagaraMeshRendererProperties, rendererIndex=1}
   {rendererClass=&NiagaraMeshRendererProperties, rendererIndex=2}
   {rendererClass=&NiagaraMeshRendererProperties, rendererIndex=3}
   {rendererClass=&NiagaraMeshRendererProperties, rendererIndex=4}
   {rendererClass=&NiagaraMeshRendererProperties, rendererIndex=5}
  ]
 }
 Flash {
  emitterUpdateScript [
   &EmitterState(Life Cycle Mode=::NewEnumerator1, Loop Behavior=::NewEnumerator1, Loop Duration=1, UseLoopDelay=false)
   &SpawnBurst_Instantaneous(Spawn Count=1, Use Spawn Probability=false)
  ]
  particleSpawnScript [
   &InitializeParticle(
    Color=<rgba 10 6.4069 0.6771 0.5>
    Lifetime=0.3
    Position Offset=<0 -90 0>
    Sprite Rotation Mode=::NewEnumerator2
    Sprite Size Mode=::NewEnumerator3
    Uniform Sprite Size=1000
    UsePositionOffset=true
    Write Material Random=true
   )
   &DynamicMaterialParameters(
    Index 0 Param 1=5
    Index 0 Param 2=200
    Index 0 Param 3=1
    Index 0 Param 4=1
    Module.Param0WriteEnabled=true
    Module.Param1WriteEnabled=true
    Module.Param2WriteEnabled=true
    Module.Param3WriteEnabled=true
   )
  ]
  particleUpdateScript [
   &ParticleState()
   &ScaleSpriteSize(
    Scale Sprite Size Mode=::NewEnumerator0
    Uniform Scale Factor=&FloatFromCurve(FloatCurve=DataInterface() {propertyValues {Curve {keys [<at 0 1>,<at 1 0>]}}}, Scale Curve=1.1)
   ) {module=ScaleSpriteSize001}
   &SolveForcesAndVelocity(Clamp Velocity=false, Limit Acceleration=false)
   &ScaleColor(
    Scale Alpha=&FloatFromCurve(
     FloatCurve=DataInterface() {propertyValues {Curve {keys [<at 0 0 RCIM_Cubic>,<at 0.2 0.6735 RCIM_Cubic>,<at 1 0>]}}}
     Scale Curve=&FN_LOD_Distance_Remap(Start Distance=300, End Distance=500, Minimum Output=0, Maximum Output=1, Clamp Output=true)
    )
    ScaleA=true
    ScaleRGB=true
   )
   &ScaleSpriteSize(Scale Factor=&Vector2DFromFloat(Value=&FloatFromCurve(FloatCurve=DataInterface() {propertyValues {Curve {keys [<at 0 1.5 RCIM_Cubic>,<at 1 0>]}}}))) {enabled=false}
   &CameraOffset(Camera Offset Amount=50)
  ]
  rendererClasses [&NiagaraSpriteRendererProperties]
  renderers [{rendererClass=&NiagaraSpriteRendererProperties}]
 }
 Glow {
  emitterUpdateScript [
   &EmitterState(Life Cycle Mode=::NewEnumerator1, Loop Behavior=::NewEnumerator1, Loop Delay=0.1, Loop Duration Mode=::NewEnumerator1, UseLoopDelay=false)
   &SpawnBurst_Instantaneous(Spawn Count=1, Use Spawn Probability=false)
  ]
  particleSpawnScript [
   &InitializeParticle(
    Color=<rgba 0.151 0.0578 0 2>
    Position Offset=<0 -90 0>
    Sprite Size Mode=::NewEnumerator3
    Uniform Sprite Size=1200
    UsePositionOffset=true
    Write Material Random=true
   )
   &DynamicMaterialParameters(
    Index 0 Param 1=5
    Index 0 Param 2=15
    Index 0 Param 3=1
    Index 0 Param 4=1
    Index 1 Param 1=&RandomRangeFloat(Minimum=0.4, Maximum=0.6)
    Index 1 Param 2=1
    Index 1 Param 3=1
    Index 1 Param 4=1
    Module.Param0WriteEnabled=true
    Module.Param0WriteEnabled1=true
    Module.Param1WriteEnabled=true
    Module.Param1WriteEnabled1=true
    Module.Param2WriteEnabled=true
    Module.Param2WriteEnabled1=true
    Module.Param3WriteEnabled=true
    Module.Param3WriteEnabled1=true
    Write Parameter Index 1=true
   )
   &CameraOffset(Camera Offset Amount=&MakeCustomFloatFromBool(Bool=$User.bIsFrontEndPreview, True Float=1, False Float=30))
  ]
  particleUpdateScript [
   &ParticleState(Kill Particles When Lifetime Has Elapsed=false)
   &ScaleSpriteSize(Scale Sprite Size Mode=::NewEnumerator0, Uniform Scale Factor=$System.SizeScale02)
   &ScaleColor(
    Scale Alpha=&Multiply_Float(
     A=0.7
     B=&MakeCustomFloatFromBool(
      Bool=&Boolean_AndOperation(
       Bool1=&Boolean_AndOperation(Bool1=$System.IsLeftCollide, Bool2=$System.IsUpCollide)
       Bool2=&Boolean_AndOperation(Bool1=$System.IsRightCollide, Bool2=$System.IsDownCollide)
      )
      True Float=0
      False Float=1
     )
    )
    ScaleA=true
    ScaleRGB=true
   )
   &FN_Sprite_Attributes(
    Color=&MakeLinearColorFromVectorAndFloat(
     "Vector (RGB)"=&Multiply_Vector(A=&MakeVectorFromLinearColorRGB(Color=<rgba 1 0.6146 0 1>), B=<1 1 1>)
     "Float (Alpha)"=&Multiply_Float(
      A=&MakeCustomFloatFromBool(
       Bool=&Boolean_AndOperation(
        Bool1=&Boolean_AndOperation(Bool1=$System.IsLeftCollide, Bool2=$System.IsUpCollide)
        Bool2=&Boolean_AndOperation(Bool1=$System.IsRightCollide, Bool2=$System.IsDownCollide)
       )
       True Float=0
       False Float=1
      )
      B=&FN_LOD_Distance_Remap(Start Distance=200, End Distance=400, Minimum Output=0.1, Maximum Output=1, Clamp Output=true)
     )
    )
    Create Color Attribute=true
    Create SpriteSize Attribute=true
    SpriteSize=&Multiply_Vector2D(A=<500 500>, B=&Multiply_Vector2D_ByFloat(Float=$System.SizeScale02, Vector2D=<1 1>))
    Write to Module Namespace=true
   )
  ]
  rendererClasses [&NiagaraSpriteRendererProperties]
  renderers [
   {rendererClass=&NiagaraSpriteRendererProperties}
   {rendererClass=&NiagaraSpriteRendererProperties, rendererIndex=1}
  ]
 }
 SmokeBlackDetail {
  emitterUpdateScript [
   &EmitterState()
   &SpawnRate(SpawnRate=40, Use Spawn Probability=false)
   &SpawnBurst_Instantaneous(Spawn Count=20, Use Spawn Probability=false, UseLoopCountLimit=false)
  ]
  particleSpawnScript [
   &PartitionParticles(Number Of Partitions=4)
   &InitializeParticle(
    Color=&MakeCustomLinearColorFromBool(
     Bool=&RandomBool()
     True Color=&RandomRangeLinearColor(Minimum=<rgba 0.7 0.3508 0.0011 2>, Maximum=<rgba 0.125 0.0633 0 1>)
     False Color=<rgba 0.0312 0.0045 0 1>
    )
    Lifetime Max=0.8
    Lifetime Min=0.7
    Lifetime Mode=::NewEnumerator1
    Position Offset=<0 -80 0>
    Sprite Rotation Mode=::NewEnumerator2
    Sprite Size Max=<60 150>
    Sprite Size Min=<100 100>
    Sprite Size Mode=::NewEnumerator2
    Sprite UV Mode=::NewEnumerator3
    UsePositionOffset=true
    Write Material Random=true
   )
   &CameraOffset(Camera Offset Amount=30)
   &AddVelocity(
    Velocity=&Multiply_Vector(
     A=&SelectVectorFromArray(
      Direct Array Index=$Particles.PartitionParticles.Partition
      Vector Selection Array=DataInterface() {propertyValues {FloatData [<-100 0 0>, <0 0 100>, <100 0 0>, <0 0 -100>]}}
     )
     B=&MergeHelper_VectorSwitch(
      0=&MakeCustomVectorFromBool(Bool=$System.IsLeftCollide, True Vector=<0 0 0>, False Vector=<1 1 1>)
      1=&MakeCustomVectorFromBool(Bool=$System.IsUpCollide, True Vector=<0 0 0>, False Vector=<1 1 1>)
      2=&MakeCustomVectorFromBool(Bool=$System.IsRightCollide, True Vector=<0 0 0>, False Vector=<1 1 1>)
      3=&MakeCustomVectorFromBool(Bool=$System.IsDownCollide, True Vector=<0 0 0>, False Vector=<1 1 1>)
      Index=$Particles.PartitionParticles.Partition
      Num=::NewEnumerator2
     )
    )
    Velocity Speed Scale=2.5
   )
   &AddVelocity(Velocity=<0 50 0>) {enabled=false, module=AddVelocity001}
  ]
  particleUpdateScript [
   &ParticleState()
   &SpriteRotationRate(Rotation Rate=&RandomRangeFloat(Minimum=-50, Maximum=50))
   &ScaleSpriteSize(
    Scale Sprite Size Mode=::NewEnumerator0
    Uniform Scale Factor=&FloatFromCurve(FloatCurve=DataInterface() {propertyValues {Curve {keys [<at 0 0.8>,<at 1 1.2>]}}}, Scale Curve=2)
   )
   &SubUVAnimation(
    End Frame Range Override=30
    "Sprite Renderer "=DataInterface() {propertyValues {SpriteRenderer=&self:SmokeBlackDetail_6.NiagaraSpriteRendererProperties_0}}
    Start Frame Range Override=&RandomRangeInt(Minimum=0, Maximum=5)
    UseEndFrame=true
    UseStartFrame=true
   )
   &ScaleColor(
    Scale Alpha=&Multiply_Float(
     A=&RandomRangeFloat(Minimum=0.5, Maximum=1)
     B=&FloatFromCurve(
      FloatCurve=DataInterface() {propertyValues {Curve {keys [
          {interpMode=RCIM_Cubic, tangentMode=RCTM_User, time=0, value=0}
          {interpMode=RCIM_Cubic, tangentMode=RCTM_User, time=0.5, value=1}
          {tangentMode=RCTM_User, time=1, value=0}
         ]}}}
      Scale Curve=&Multiply_Float(
       A=1.5
       B=&MakeCustomFloatFromBool(
        Bool=&Boolean_AndOperation(
         Bool1=&Boolean_AndOperation(Bool1=$System.IsLeftCollide, Bool2=$System.IsUpCollide)
         Bool2=&Boolean_AndOperation(Bool1=$System.IsRightCollide, Bool2=$System.IsDownCollide)
        )
        True Float=0
        False Float=1
       )
      )
     )
    )
    ScaleA=true
    ScaleRGB=true
   )
   &LinearForce(Force=<0 500 0>)
   &SolveForcesAndVelocity(Clamp Velocity=false, Limit Acceleration=false)
  ]
  rendererClasses [&NiagaraSpriteRendererProperties]
  renderers [{rendererClass=&NiagaraSpriteRendererProperties}]
 }
 Sparks {
  simTarget=GPUComputeSim
  emitterUpdateScript [
   &EmitterState(Delay First Loop Only=true, Life Cycle Mode=::NewEnumerator1, Loop Delay=0.2, Loop Duration=0.6, UseLoopDelay=true)
   &SpawnBurst_Instantaneous(
    Spawn Count=10
    Spawn Probability=&MakeCustomFloatFromBool(
     Bool=&Boolean_OrOperation(
      Bool1=&Boolean_AndOperation(Bool1=$System.IsLeftCollide, Bool2=$System.IsUpCollide)
      Bool2=&Boolean_AndOperation(Bool1=$System.IsRightCollide, Bool2=$System.IsDownCollide)
     )
     True Float=0
     False Float=1
    )
    Use Spawn Probability=true
    UseLoopCountLimit=false
   )
   &SpawnRate(
    Spawn Probability=&MakeCustomFloatFromBool(
     Bool=&Boolean_AndOperation(
      Bool1=&Boolean_AndOperation(Bool1=$System.IsLeftCollide, Bool2=$System.IsUpCollide)
      Bool2=&Boolean_AndOperation(Bool1=$System.IsRightCollide, Bool2=$System.IsDownCollide)
     )
     True Float=0
     False Float=1
    )
    SpawnRate=300
    Use Spawn Probability=true
   )
  ]
  particleSpawnScript [
   &InitializeParticle(
    Color=<rgba 10 8.855 0 1>
    Lifetime Max=1
    Lifetime Min=0.5
    Lifetime Mode=::NewEnumerator1
    Sprite Size Max=<0.3 2>
    Sprite Size Min=<0.2 2>
    Sprite Size Mode=::NewEnumerator2
    UsePositionOffset=false
    Write Material Random=true
   )
   &ShapeLocation(MaskSpawnGroup=false, Offset=<0 70 0>, Offset Mode=::NewEnumerator0, Override Seed=false, UseOwnerScale=true)
   &AddVelocity(Velocity=<0 0 0>) {enabled=false}
   &AddVelocity(
    Cone Angle=0
    Cone Axis=<0 1 0>
    Module.UseConeFalloff=false
    Module.UseVelDistribution=false
    Override Seed=false
    Velocity Mode=::NewEnumerator2
    Velocity Speed=&RandomRangeFloat(Minimum=1000, Maximum=1000)
   ) {module=AddVelocity001}
   &InheritVelocity(Speed Limit=true, UseSpeedThreshold=false) {enabled=false}
  ]
  particleUpdateScript [
   /SetVariables(Particles.SizeAll:Vector2f=&Multiply_Vector2D_ByFloat(
     Float=&FN_LOD_Distance_Remap(Start Distance=1000, End Distance=2000, Minimum Output=0.8, Maximum Output=1.7, Clamp Output=true)
     Vector2D=<2 2>
    )) {isSetParameters=true}
   &ParticleState()
   &ScaleSpriteSize(Scale Factor=&Multiply_Vector2D(A=<2 2>, B=$Particles.SizeAll), Scale Sprite Size Mode=::NewEnumerator2)
   &CurlNoiseForce(Bias Noise=false, Noise Frequency=30, Noise Strength=2000, Pan Noise=false)
   &Drag(Use Linear Drag=true, Use Rotational Drag=true)
   &InheritSourceMovement(
    Applied Position Scale=&Multiply_VectorByFloat(Vector=<1 1 1>, Float=&RandomRangeFloat(Minimum=0.2, Maximum=0.7))
    ApplyForce=false
    ApplyPosition=true
    ApplyPositionFalloff=false
   )
   &SolveForcesAndVelocity(Clamp Velocity=false, Limit Acceleration=false)
   &ScaleSpriteSizeBySpeed(Max Scale Factor=<1 2>, Min Scale Factor=<0.5 0.5>) {enabled=false}
   &FN_Sprite_Attributes(
    Color=<rgba 1 0.7033 0.1099 0.1>
    Create Color Attribute=true
    Create SpriteSize Attribute=true
    SpriteSize=&Multiply_Vector2D(A=$Particles.SpriteSize, B=<20 4>)
    Write to Module Namespace=true
   )
   &DynamicMaterialParameters(
    Index 0 Param 1=2
    Module.Index0Param0Write=true
    Module.Index0Param1Write=true
    Module.Index0Param2Write=true
    Module.Index0Param3Write=true
   )
  ]
  rendererClasses [&NiagaraSpriteRendererProperties]
  renderers [
   {rendererClass=&NiagaraSpriteRendererProperties}
   {rendererClass=&NiagaraSpriteRendererProperties, rendererIndex=1}
  ]
 }
 Spiral_AirFlow {
  emitterUpdateScript [
   &EmitterState()
   &SpawnRate(SpawnRate=10, Use Spawn Probability=false)
   &SpawnBurst_Instantaneous(Use Spawn Probability=false, UseLoopCountLimit=false)
  ]
  particleSpawnScript [
   &InitializeParticle(Lifetime=0.4, Mesh Scale Mode=::NewEnumerator3, Mesh Uniform Scale=1.5, UsePositionOffset=false, Write Material Random=true)
   &CameraOffset(Camera Offset Amount=200)
   &InitialMeshOrientation(Mesh Orientation Mode=::NewEnumerator2, Rotation=<0 0.25 0>, Use Rotation Vector=true)
   &InitialMeshOrientation(
    Mesh Orientation Mode=::NewEnumerator2
    Rotation=&RandomRangeVector(Minimum=<0 0 0>, Maximum=<0 1 0>)
    Use Rotation Vector=true
   ) {module=InitialMeshOrientation001}
   &SystemLocation(MaskSpawnGroup=false)
  ]
  particleUpdateScript [
   &ParticleState()
   &UpdateMeshOrientation(Rotation Rate=-0.02, Rotation Vector=<0 1 0>)
   &ScaleMeshSize(Scale Factor=&Multiply_VectorByFloat(Vector=<1 1 1>, Float=&Multiply_Float(A=0.5, B=1.5)))
   &ScaleColor(
    Scale Alpha=&FloatFromCurve(
     FloatCurve=DataInterface() {propertyValues {Curve {keys [
         {interpMode=RCIM_Cubic, tangentMode=RCTM_User, time=0, value=0}
         {interpMode=RCIM_Cubic, tangentMode=RCTM_User, time=0.5, value=1}
         {tangentMode=RCTM_User, time=1, value=0}
        ]}}}
     Scale Curve=&Multiply_Float(
      A=0.5
      B=&MakeCustomFloatFromBool(
       Bool=&Boolean_AndOperation(
        Bool1=&Boolean_AndOperation(Bool1=$System.IsLeftCollide, Bool2=$System.IsUpCollide)
        Bool2=&Boolean_AndOperation(Bool1=$System.IsRightCollide, Bool2=$System.IsDownCollide)
       )
       True Float=0
       False Float=1
      )
     )
    )
    ScaleA=true
    ScaleRGB=true
   )
  ]
  rendererClasses [&NiagaraMeshRendererProperties]
  renderers [{rendererClass=&NiagaraMeshRendererProperties, rendererIndex=0}]
 }
 TrailMesh {
  emitterUpdateScript [
   &EmitterState(Life Cycle Mode=::NewEnumerator1, Loop Behavior=::NewEnumerator1, Loop Duration=1, UseLoopDelay=true)
   &SpawnBurst_Instantaneous(Spawn Count=2, Use Spawn Probability=false, UseLoopCountLimit=false)
  ]
  particleSpawnScript [
   &PartitionParticles()
   &InitializeParticle(
    Color=<rgba 5 4.7499 2.7291 1>
    Mesh Scale=&Multiply_VectorByFloat(Vector=<1 1 1>, Float=1.3)
    Mesh Scale Mode=::NewEnumerator4
    Position Offset=<0 -100 0>
    UsePositionOffset=true
    Write Material Random=true
   )
   &InitialMeshOrientation(Mesh Orientation Mode=::NewEnumerator7, Rotation=<0 0 0.5>, Use Rotation Vector=true)
   &InitialMeshOrientation(Mesh Orientation Mode=::NewEnumerator7, Rotation=<0 0 -0.25>, Use Rotation Vector=true) {module=InitialMeshOrientation001}
   &InitialMeshOrientation(
    Mesh Orientation Mode=::NewEnumerator2
    Rotation=&SelectVectorFromArray(
     Direct Array Index=$Particles.PartitionParticles.Partition
     Vector Selection Array=DataInterface() {propertyValues {FloatData [<0 0 0>, <0 0.25 0>]}}
    )
    Use Rotation Vector=true
   ) {module=InitialMeshOrientation002}
  ]
  particleUpdateScript [
   &Drag(Use Linear Drag=true, Use Rotational Drag=true)
   &SolveForcesAndVelocity(Clamp Velocity=false, Limit Acceleration=false)
   &ParticleState(Kill Particles When Lifetime Has Elapsed=false)
   &ScaleMeshSize(Scale Factor=&VectorFromCurve(
     VectorCurve=DataInterface() {propertyValues {XCurve {keys [<at 0 1>,<at 1 1>]}, YCurve {keys [<at 0 1>,<at 1 1>]}, ZCurve {keys [<at 0 1>,<at 1 1>]}}}
     Scale Curve=<1.2 1 1>
    )) {module=ScaleMeshSize001}
   &ScaleMeshSize(Scale Factor=&Multiply_VectorByFloat(Vector=<1 1 1>, Float=$System.SizeScale02))
   &ScaleColor(
    Scale Alpha=&MakeCustomFloatFromBool(
     Bool=&Boolean_OrOperation(
      Bool1=&Boolean_AndOperation(Bool1=$System.IsLeftCollide, Bool2=$System.IsUpCollide)
      Bool2=&Boolean_AndOperation(Bool1=$System.IsRightCollide, Bool2=$System.IsDownCollide)
     )
     True Float=0
     False Float=1
    )
    Scale RGB=&MakeVectorFromLinearColorRGB(Color=<rgba 0.5 0.3791 0.1517 1>)
    ScaleA=true
    ScaleRGB=true
   )
   &DynamicMaterialParameters(
    Index 0 Param 1=2
    Index 0 Param 2=100
    Module.Index0Param0Write=true
    Module.Index0Param1Write=true
    Module.Index0Param2Write=true
    Module.Index0Param3Write=true
   )
  ]
  rendererClasses [&NiagaraMeshRendererProperties]
  renderers [{rendererClass=&NiagaraMeshRendererProperties, rendererIndex=0}]
 }
}
summary {
 systemName=NS_SinisterMelon_Projectile
 emitters [
  {emitterName=Cross, rendererClasses [&NiagaraMeshRendererProperties]}
  {emitterName=TrailMesh, rendererClasses [&NiagaraMeshRendererProperties]}
  {emitterName=Airflow, rendererClasses [&NiagaraMeshRendererProperties]}
  {emitterName=Glow, rendererClasses [&NiagaraSpriteRendererProperties]}
  {emitterName=Spiral_AirFlow, rendererClasses [&NiagaraMeshRendererProperties]}
  {emitterName=SmokeBlackDetail, rendererClasses [&NiagaraSpriteRendererProperties]}
  {emitterName=Flash, rendererClasses [&NiagaraSpriteRendererProperties]}
  {emitterName=Sparks, simTarget=GPUComputeSim, rendererClasses [&NiagaraSpriteRendererProperties]}
 ]
 userVariables [
  User.ActiveArray {
   type=NiagaraDataInterfaceArrayBool
   description=""
   defaultValue {
    struct=NiagaraExt_VariableValue_DataInterface
    value {
     dataInterface=&self:NiagaraDataInterfaceArrayBool_0
     dataInterfaceClass=/Script/Niagara.NiagaraDataInterfaceArrayBool
    }
   }
  }
  User.bIsFrontEnd {type=NiagaraBool}
  User.bIsFrontEndPreview {type=NiagaraBool}
  User.Physical Surface Dust {type=NiagaraBool}
  User.VariantSwitch {type=NiagaraFloat, defaultValue=0, description=""}
 ]
}
