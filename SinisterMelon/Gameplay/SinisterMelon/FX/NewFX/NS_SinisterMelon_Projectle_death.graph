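# Niagara burst effect for the death of a Sinister Melon projectile, built from orange sparks, a brief flare and smoke.
# GPU sparks with curl noise and animated flipbook smoke are partitioned into four streams along ±X and ±Z, slowed by drag.
# The flare fades with age and scales alpha by distance.
# All emitters share a system camera offset and zero their alpha when all four directional collide flags are set, unless FX preview overrides it.
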NiagaraSystem /SinisterMelon/Gameplay/SinisterMelon/FX/NewFX/NS_SinisterMelon_Projectle_death
dependencies {
 usedDataInterfaces [
  &NiagaraDataInterfaceArrayFloat3
  &NiagaraDataInterfaceCurve
  &NiagaraDataInterfaceVector2DCurve
  &NiagaraDataInterfaceSpriteRendererInfo
  &NiagaraDataInterfaceColorCurve
 ]
 usedDynamicInputs [&RandomRangeVector2D]
 usedModules [
  &self:SystemSpawnScript.SystemScriptSource.SystemScriptGraph.NiagaraNodeAssignment_0.SetVariables_#1
  &SystemState
  &self:CheckHidingState
  &self:SystemSpawnScript.SystemScriptSource.SystemScriptGraph.NiagaraNodeAssignment_1.SetVariables_#2
 ]
}
emitters {
 Empty012 {
  simTarget=GPUComputeSim
  emitterUpdateScript [
   &EmitterState()
   &SpawnBurst_Instantaneous(Spawn Count=500, Use Spawn Probability=false, UseLoopCountLimit=false)
  ]
  particleSpawnScript [
   &PartitionParticles(Number Of Partitions=4)
   &InitializeParticle(
    Color=<rgba 10 5.2591 0.0614 1>
    Lifetime Max=0.7
    Lifetime Min=0.3
    Lifetime Mode=::NewEnumerator1
    Sprite Size Max=<0.3 2>
    Sprite Size Min=<0.2 2>
    Sprite Size Mode=::NewEnumerator2
    UsePositionOffset=false
    Write Material Random=true
   )
   &ShapeLocation(Box Size=<10 50 10>, MaskSpawnGroup=false, Override Seed=false, Shape Primitive=::NewEnumerator2, Surface Only=false, UseOwnerScale=true)
   &AddVelocity(
    Velocity=&SelectVectorFromArray(
     Direct Array Index=$Particles.PartitionParticles.Partition
     Vector Selection Array=DataInterface() {propertyValues {FloatData [<0 0 100>, <0 0 -100>, <100 0 0>, <-100 0 0>]}}
    )
    Velocity Speed Scale=&RandomRangeFloat(Minimum=2, Maximum=8)
   ) {module=AddVelocity001}
  ]
  particleUpdateScript [
   &ParticleState()
   &ScaleSpriteSize(
    Scale Sprite Size Mode=::NewEnumerator0
    Uniform Scale Factor=&FloatFromCurve(FloatCurve=DataInterface() {propertyValues {Curve {keys [<at 0.7 4>,<at 1 0>]}}}, Scale Curve=2)
   )
   &CurlNoiseForce(Bias Noise=false, Noise Frequency=30, Noise Strength=2000, Pan Noise=false)
   &Drag(Drag=3, Use Linear Drag=true, Use Rotational Drag=true)
   &SolveForcesAndVelocity(Clamp Velocity=false, Limit Acceleration=false)
   &FN_Sprite_Attributes(
    Color=&MakeLinearColorFromVectorAndFloat(
     "Vector (RGB)"=&MakeVectorFromLinearColorRGB(Color=<rgba 1 0.5688 0.0651 0.1>)
     "Float (Alpha)"=&Multiply_Float(
      A=0.1
      B=&Lerp_Float(
       A=&MakeCustomFloatFromBool(
        Bool=&Boolean_AndOperation(
         Bool1=&Boolean_AndOperation(Bool1=$System.IsLeftCollide, Bool2=$System.IsUpCollide)
         Bool2=&Boolean_AndOperation(Bool1=$System.IsRightCollide, Bool2=$System.IsDownCollide)
        )
        True Float=0
        False Float=1
       )
       B=1
       Alpha=$System.FX preview
      )
     )
    )
    Create Color Attribute=true
    Create SpriteSize Attribute=true
    SpriteSize=&Multiply_Vector2D(A=$Particles.SpriteSize, B=<20 4>)
    Write to Module Namespace=true
   )
   &DynamicMaterialParameters(
    Index 0 Param 1=2
    Module.Index0Param0Write=true
    Module.Index0Param1Write=true
    Module.Index0Param2Write=true
    Module.Index0Param3Write=true
   )
   &ScaleColor(
    Scale Alpha=&Lerp_Float(
     A=&MakeCustomFloatFromBool(
      Bool=&Boolean_AndOperation(
       Bool1=&Boolean_AndOperation(Bool1=$System.IsLeftCollide, Bool2=$System.IsUpCollide)
       Bool2=&Boolean_AndOperation(Bool1=$System.IsRightCollide, Bool2=$System.IsDownCollide)
      )
      True Float=0
      False Float=1
     )
     B=1
     Alpha=$System.FX preview
    )
    ScaleA=true
    ScaleRGB=true
   )
   &CameraOffset(Camera Offset Amount=$System.CameraOffsetAll)
  ]
  rendererClasses [&NiagaraSpriteRendererProperties]
  renderers [
   {rendererClass=&NiagaraSpriteRendererProperties}
   {rendererClass=&NiagaraSpriteRendererProperties, rendererIndex=1}
  ]
 }
 Flash_Flare_pink {
  emitterUpdateScript [
   &EmitterState(Inactive Response=::NewEnumerator1, Life Cycle Mode=::NewEnumerator1, Loop Behavior=::NewEnumerator1, Loop Duration=0.4, UseLoopDelay=false)
   &SpawnBurst_Instantaneous(Spawn Count=1, Use Spawn Probability=false)
  ]
  particleSpawnScript [
   &InitializeParticle(
    Color=<rgba 5 5 5 1>
    Lifetime=0.35
    Sprite Rotation Angle=-5
    Sprite Rotation Mode=::NewEnumerator0
    Sprite Size=<200 100>
    Sprite Size Mode=::NewEnumerator4
    UsePositionOffset=false
    Write Material Random=true
   ) {module=InitializeParticle001}
   /SetVariables(Particles.Rotation2=85) {isSetParameters=true}
   &DynamicMaterialParameters(
    Index 0 Param 1=1
    Index 0 Param 2=1
    Module.Index0Float0Write=true
    Module.Index0Float1Write=true
    Module.Index0Float2Write=true
    Module.Index0Float3Write=true
   ) {module=DynamicMaterialParameters001}
  ]
  particleUpdateScript [
   &ParticleState()
   &SolveForcesAndVelocity(Clamp Velocity=false, Limit Acceleration=false)
   &ScaleColor(
    Scale Alpha=&Multiply_Float(
     A=&Multiply_Float(
      A=&Lerp_Float(A=1, B=0, Alpha=$Particles.NormalizedAge)
      B=&FN_LOD_Distance_Remap(Start Distance=500, End Distance=2000, Minimum Output=0.1, Maximum Output=0.8, Clamp Output=true)
     )
     B=&Lerp_Float(
      A=&MakeCustomFloatFromBool(
       Bool=&Boolean_AndOperation(
        Bool1=&Boolean_AndOperation(Bool1=$System.IsLeftCollide, Bool2=$System.IsUpCollide)
        Bool2=&Boolean_AndOperation(Bool1=$System.IsRightCollide, Bool2=$System.IsDownCollide)
       )
       True Float=0
       False Float=1
      )
      B=1
      Alpha=$System.FX preview
     )
    )
    Scale RGB=&MakeVectorFromLinearColorRGB(Color=&ColorFromCurve(ColorCurve=DataInterface() {propertyValues {
        AlphaCurve {keys []}
        BlueCurve {keys [{arriveTangent=3.2397, interpMode=RCIM_Cubic, time=0, value=0.0069}, <at 0.3738 0.0236>, <at 0.5436 0.0014>, <at 0.8697 0>]}
        GreenCurve {keys [{arriveTangent=-1.4029, interpMode=RCIM_Cubic, time=0, value=0.1413}, <at 0.3738 0.5786>, <at 0.5436 0.4242>, <at 0.8697 0.1667>]}
        RedCurve {keys [{arriveTangent=-3.2397, interpMode=RCIM_Cubic, time=0, value=0.2}, <at 0.3738 1>, <at 0.5436 1>, <at 0.8697 0.5>]}
       }}))
    ScaleA=true
    ScaleRGB=true
   )
   &ScaleSpriteSize(Scale Factor=&Vector2DFromCurve(
     Vector2Curve=DataInterface() {propertyValues {XCurve {keys [<at 0 1>,<at 1 1>]}, YCurve {keys [<at 0 1>,<at 1 0.3>]}}}
     ScaleCurve=&Multiply_Vector2D_ByFloat(Float=5, Vector2D=<1.1 1.1>)
    ))
   &CameraOffset(Camera Offset Amount=$System.CameraOffsetAll)
  ]
  rendererClasses [&NiagaraSpriteRendererProperties]
  renderers [
   {rendererClass=&NiagaraSpriteRendererProperties}
   {rendererClass=&NiagaraSpriteRendererProperties, rendererIndex=1}
  ]
 }
 SmokeBlackDetail {
  emitterUpdateScript [
   &EmitterState()
   &SpawnBurst_Instantaneous(Spawn Count=40, Use Spawn Probability=false, UseLoopCountLimit=false)
  ]
  particleSpawnScript [
   &PartitionParticles(Number Of Partitions=4)
   &InitializeParticle(
    Color=&MakeCustomLinearColorFromBool(
     Bool=&RandomBool()
     True Color=&RandomRangeLinearColor(Minimum=<rgba 1 0.5011 0.0016 2>, Maximum=<rgba 0.125 0.0633 0 1>)
     False Color=<rgba 0.401 0.2004 0.0302 1>
    )
    Lifetime Max=0.4
    Lifetime Min=0.6
    Lifetime Mode=::NewEnumerator1
    Sprite Rotation Mode=::NewEnumerator2
    Sprite Size Max=<60 150>
    Sprite Size Min=<100 100>
    Sprite Size Mode=::NewEnumerator2
    Sprite UV Mode=::NewEnumerator3
    UsePositionOffset=true
    Write Material Random=true
   )
   &AddVelocity(
    Velocity=&SelectVectorFromArray(
     Direct Array Index=$Particles.PartitionParticles.Partition
     Vector Selection Array=DataInterface() {propertyValues {FloatData [<0 0 100>, <0 0 -100>, <100 0 0>, <-100 0 0>]}}
    )
    Velocity Speed Scale=&RandomRangeFloat(Minimum=2, Maximum=8)
   )
  ]
  particleUpdateScript [
   &CameraOffset(Camera Offset Amount=$System.CameraOffsetAll)
   &ParticleState()
   &SpriteRotationRate(Rotation Rate=&RandomRangeFloat(Minimum=-50, Maximum=50))
   &ScaleSpriteSize(
    Scale Sprite Size Mode=::NewEnumerator0
    Uniform Scale Factor=&FloatFromCurve(FloatCurve=DataInterface() {propertyValues {Curve {keys [<at 0 0.8>,<at 1 1.2>]}}}, Scale Curve=3)
   )
   &SubUVAnimation(
    End Frame Range Override=30
    "Sprite Renderer "=DataInterface() {propertyValues {SpriteRenderer=&self:SmokeBlackDetail_6.NiagaraSpriteRendererProperties_0}}
    Start Frame Range Override=&RandomRangeInt(Minimum=0, Maximum=5)
    UseEndFrame=true
    UseStartFrame=true
   )
   &ScaleColor(
    Scale Alpha=&Multiply_Float(
     A=&RandomRangeFloat(Minimum=0.5, Maximum=1)
     B=&FloatFromCurve(
      FloatCurve=DataInterface() {propertyValues {Curve {keys [
          {interpMode=RCIM_Cubic, tangentMode=RCTM_User, time=0, value=0}
          {interpMode=RCIM_Cubic, tangentMode=RCTM_User, time=0.5, value=1}
          {tangentMode=RCTM_User, time=1, value=0}
         ]}}}
      Scale Curve=&Multiply_Float(
       A=1.5
       B=&Lerp_Float(
        A=&MakeCustomFloatFromBool(
         Bool=&Boolean_AndOperation(
          Bool1=&Boolean_AndOperation(Bool1=$System.IsLeftCollide, Bool2=$System.IsUpCollide)
          Bool2=&Boolean_AndOperation(Bool1=$System.IsRightCollide, Bool2=$System.IsDownCollide)
         )
         True Float=0
         False Float=1
        )
        B=1
        Alpha=$System.FX preview
       )
      )
     )
    )
    ScaleA=true
    ScaleRGB=true
   )
   &Drag(Drag=&RandomRangeFloat(Minimum=0.5, Maximum=3), Use Linear Drag=true, Use Rotational Drag=true)
   &SolveForcesAndVelocity(Clamp Velocity=false, Limit Acceleration=false)
  ]
  rendererClasses [&NiagaraSpriteRendererProperties]
  renderers [{rendererClass=&NiagaraSpriteRendererProperties}]
 }
}
summary {
 systemName=NS_SinisterMelon_Projectle_death
 emitters [
  {emitterName=SmokeBlackDetail, rendererClasses [&NiagaraSpriteRendererProperties]}
  {emitterName=Empty012, simTarget=GPUComputeSim, rendererClasses [&NiagaraSpriteRendererProperties]}
  {emitterName=Flash_Flare_pink, rendererClasses [&NiagaraSpriteRendererProperties]}
 ]
 userVariables [User.ActiveArray {
   type=NiagaraDataInterfaceArrayBool
   description=""
   defaultValue {
    struct=NiagaraExt_VariableValue_DataInterface
    value {
     dataInterface=&self:NiagaraDataInterfaceArrayBool_0
     dataInterfaceClass=/Script/Niagara.NiagaraDataInterfaceArrayBool
    }
   }
  }]
}
