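# Niagara effect for SinisterMelon's first big slash: warm orange, glowing meshes sweep around with decelerating spin and fade.
# Slash, DetalEnergy and SlashCenter layer burst-spawned meshes with curve-driven color and material parameters, sharing a system camera offset.
# LeakParticle adds a delayed GPU burst of vortex-swirled sprites.
# PPV drives a component renderer, apparently post-process, spawning only when User.bIsLocalPlayer is true and disabling it just before particle death.
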NiagaraSystem /SinisterMelon/Gameplay/SinisterMelon/FX/NewFX/NS_SinisterMelon_SlashBigFirst
dependencies {
 usedDataInterfaces [&NiagaraDataInterfaceCurve, &NiagaraDataInterfaceColorCurve]
 usedDynamicInputs []
 usedModules [
  &self:SystemSpawnScript.SystemScriptSource.SystemScriptGraph.NiagaraNodeAssignment_0.SetVariables_#1
  &self:SystemSpawnScript.SystemScriptSource.SystemScriptGraph.NiagaraNodeAssignment_1.SetVariables_#2
  &self:SystemSpawnScript.SystemScriptSource.SystemScriptGraph.NiagaraNodeAssignment_2.SetVariables_#3
  &SystemState
 ]
}
emitters {
 DetalEnergy {
  emitterUpdateScript [
   &EmitterState(Life Cycle Mode=::NewEnumerator1, Loop Behavior=::NewEnumerator1, Loop Duration=1, UseLoopDelay=false)
   &SpawnBurst_Instantaneous(Use Spawn Probability=false, UseLoopCountLimit=false)
  ]
  particleSpawnScript [
   &InitializeParticle(
    Color=<rgba 5 2.8654 0 1>
    Lifetime=1.8
    Mesh Scale=&Multiply_VectorByFloat(Vector=<1 1 0.3>, Float=1.8)
    Mesh Scale Mode=::NewEnumerator4
    Position Offset=<0 0 5>
    UsePositionOffset=true
    Write Material Random=true
   )
   /SetVariables(Particles.MeshOrientation=$Engine.Owner.Rotation) {enabled=false, isSetParameters=true, module=SetVariables_#4}
   &CameraOffset(Camera Offset Amount=$System.CameraOffsetAll)
  ]
  particleUpdateScript [
   &ParticleState()
   &ScaleMeshSize(Scale Factor=&Multiply_Vector(
     A=$Engine.Owner.Scale
     B=&Multiply_Vector(
      A=$System.SizeScale
      B=&MakeVector(
       X=1
       Y=1
       Z=&FloatFromCurve(FloatCurve=DataInterface() {propertyValues {Curve {keys [<at 0 0.3>,<at 1 0.3>]}}}, Scale Curve=2)
      )
     )
    ))
   &UpdateMeshOrientation(
    Rotation Rate=&FloatFromCurve(
     FloatCurve=DataInterface() {propertyValues {Curve {keys [{arriveTangent=-3.0754, interpMode=RCIM_Cubic, leaveTangent=-3.0754, tangentMode=RCTM_User, time=0, value=2}, <at 0.5 0.2>]}}}
     Scale Curve=0.35
    )
    Rotation Vector=<0 0 -1.3>
   ) {module=UpdateMeshOrientation001}
   /SetVariables() {enabled=false, isSetParameters=true, module=SetVariables_#5}
   &DynamicMaterialParameters(
    Index 0 Param 1=&FloatFromCurve(FloatCurve=DataInterface() {propertyValues {Curve {keys [<at 0 0 RCIM_Cubic>,<at 1 0>]}}})
    Index 0 Param 2=0.1
    Index 0 Param 4=15
    Module.Index0Param0Write=true
    Module.Index0Param1Write=true
    Module.Index0Param2Write=true
    Module.Index0Param3Write=true
   ) {module=DynamicMaterialParameters001}
   &ScaleColor(
    Scale Alpha=&FloatFromCurve(FloatCurve=DataInterface() {propertyValues {Curve {keys [<at 0 1>,<at 1 0>]}}})
    Scale RGB=&Multiply_VectorByFloat(
     Vector=&MakeVectorFromLinearColorRGB(Color=&ColorFromCurve(ColorCurve=DataInterface() {propertyValues {
         AlphaCurve {keys []}
         BlueCurve {keys [<at 0 0.7457 RCIM_Cubic>]}
         GreenCurve {keys [<at 0 0.9788 RCIM_Cubic>]}
         RedCurve {keys [<at 0 1 RCIM_Cubic>]}
        }}))
     Float=&FloatFromCurve(
      FloatCurve=DataInterface() {propertyValues {Curve {keys [<at 0 1>, {arriveTangent=-1, interpMode=RCIM_Cubic, leaveTangent=-1, time=0.7, value=1}, <at 1 0>]}}}
      Scale Curve=3
     )
    )
    ScaleA=true
    ScaleRGB=true
   )
  ]
  rendererClasses [&NiagaraMeshRendererProperties]
  renderers [{rendererClass=&NiagaraMeshRendererProperties, rendererIndex=0}]
 }
 LeakParticle {
  simTarget=GPUComputeSim
  emitterUpdateScript [
   &EmitterState(Life Cycle Mode=::NewEnumerator1, Loop Behavior=::NewEnumerator1, Loop Delay=0.2, Loop Duration=1, UseLoopDelay=true)
   &SpawnBurst_Instantaneous(Spawn Count=100, Use Spawn Probability=false, UseLoopCountLimit=false)
  ]
  particleSpawnScript [
   &PartitionParticles()
   &InitializeParticle(
    Color=<rgba 2 0.3757 0 1>
    Lifetime Max=1.5
    Lifetime Min=2
    Lifetime Mode=::NewEnumerator1
    Sprite Size Max=<0.5 6>
    Sprite Size Min=<0.3 2>
    Sprite Size Mode=::NewEnumerator2
    UsePositionOffset=false
    Write Material Random=true
   )
   &ShapeLocation(
    Handle Radius=70
    Large Radius=120
    MaskSpawnGroup=false
    Offset=&RandomRangeVector(Minimum=<0 0 -20>, Maximum=<0 0 20>)
    Offset Mode=::NewEnumerator0
    Override Seed=false
    Rotation Angle=0.4
    Rotation Axis=<0 0 1>
    Rotation Mode=::NewEnumerator1
    Shape Primitive=::NewEnumerator3
    U Distribution=0.55
    UseOwnerScale=true
    V Distribution=1
   )
   &InitialMeshOrientation(
    Mesh Orientation Mode=::NewEnumerator6
    Orientation Quaternion:Quat4f=&FindQuatBetweenTwoVectors(
     From Vector=<0 0 1>
     To Vector=&RotateVectorByQuaternion(Vector To Rotate=<0 1 0>, Quaternion=$Particles.SkeletalMeshLocation001.SampledRotation)
    )
    Use Rotation Vector=false
   )
   &AddVelocity(
    Module.Use Falloff Distance=false
    Override Seed=false
    Velocity Mode=::NewEnumerator1
    Velocity Speed=&RandomRangeFloat(Minimum=50, Maximum=200)
   ) {enabled=false}
   &CameraOffset(Camera Offset Amount=$System.CameraOffsetAll) {module=CameraOffset001}
  ]
  particleUpdateScript [
   &ParticleState()
   &ScaleSpriteSize(
    Scale Factor=&MakeVector2D(
     X=&FloatFromCurve(FloatCurve=DataInterface() {propertyValues {Curve {keys [<at 0 0.2>,<at 1 1>]}}})
     Y=1
    )
    Scale Sprite Size Mode=::NewEnumerator2
   ) {module=ScaleSpriteSize002}
   &ScaleSpriteSize(
    Scale Factor=&MakeVector2D(
     X=1
     Y=&FloatFromCurve(FloatCurve=DataInterface() {propertyValues {Curve {keys [{interpMode=RCIM_Cubic, leaveTangent=-10.3145, tangentMode=RCTM_Break, time=0, value=4}, <at 1 1>]}}})
    )
    Scale Sprite Size Mode=::NewEnumerator2
   ) {module=ScaleSpriteSize001}
   &ScaleSpriteSize(Scale Sprite Size Mode=::NewEnumerator0, Uniform Scale Factor=2)
   &CurlNoiseForce(Bias Noise=false, Noise Frequency=60, Noise Strength=200, Pan Noise=true, Pan Noise Field=<0 0 1>) {enabled=false}
   &Drag(Drag=3, Use Linear Drag=true, Use Rotational Drag=true)
   &GravityForce(Gravity=<0 0 0>) {enabled=false}
   &VortexForce(
    UseOriginOffset=false
    Vortex Force Amount=&Multiply_Float(A=&RandomRangeFloat(Minimum=-2000, Maximum=-2000), B=1)
   )
   &ScaleVelocity()
   /SetVariables(Transient.PhysicsForce:Vector3f=$Transient.PhysicsForce) {enabled=false, isSetParameters=true}
   &SolveForcesAndVelocity(Clamp Velocity=false, Limit Acceleration=false)
   &ScaleSpriteSizeBySpeed(Max Scale Factor=<1 1>, Velocity Threshold=5633.8472)
   &ScaleColor(
    Scale Alpha=&FloatFromCurve(FloatCurve=DataInterface() {propertyValues {Curve {keys [
         {tangentMode=RCTM_User, time=0, value=0}
         {arriveTangent=-3.3056, interpMode=RCIM_Cubic, leaveTangent=-3.3056, tangentMode=RCTM_User, time=0.3, value=1}
         {arriveTangent=-0.0264, interpMode=RCIM_Cubic, leaveTangent=-0.0264, tangentMode=RCTM_User, time=1, value=0}
        ]}}})
    Scale RGB=&VectorFromFloat(Value=10)
    ScaleA=true
    ScaleRGB=true
   )
   &FN_Sprite_Attributes(
    Color=&MultiplyLinearColor(
     Linear Color=<rgba 0.151 0.0606 0 1>
     Scale Factor=&MakeVector4(
      X=1
      Y=1
      Z=1
      W=&FloatFromCurve(
       FloatCurve=DataInterface() {propertyValues {Curve {keys [
           {arriveTangent=-2.2465, interpMode=RCIM_Cubic, leaveTangent=-2.2465, tangentMode=RCTM_User, time=0, value=0}
           <at 0.2958 0.4971 RCIM_Cubic>
           {arriveTangent=-0.4533, interpMode=RCIM_Cubic, leaveTangent=-0.4533, tangentMode=RCTM_User, time=1, value=0}
          ]}}}
       Scale Curve=0.5816
      )
     )
    )
    Create Color Attribute=true
    Create SpriteSize Attribute=true
    SpriteSize=<100 150>
    Write to Module Namespace=true
   )
  ]
  rendererClasses [&NiagaraSpriteRendererProperties]
  renderers [
   {rendererClass=&NiagaraSpriteRendererProperties}
   {rendererClass=&NiagaraSpriteRendererProperties, rendererIndex=1}
  ]
 }
 PPV {
  emitterUpdateScript [
   &EmitterState(Life Cycle Mode=::NewEnumerator1, Loop Behavior=::NewEnumerator1, Loop Duration=0.5, UseLoopDelay=false)
   &SpawnBurst_Instantaneous(
    Spawn Probability=&MakeCustomFloatFromBool(Bool=$User.bIsLocalPlayer, True Float=1, False Float=0)
    Use Spawn Probability=true
    UseLoopCountLimit=false
   )
  ]
  particleSpawnScript [&InitializeParticle(
    Color Mode=::NewEnumerator0
    Lifetime=0.4
    Sprite Size Mode=::NewEnumerator3
    Uniform Sprite Size=200
    UsePositionOffset=false
    Write Material Random=true
   )]
  particleUpdateScript [
   /SetVariables(Particles.SpeedNormalizedB:NiagaraFloat=&FloatFromCurve(
     FloatCurve=DataInterface() {propertyValues {Curve {keys [
         {interpMode=RCIM_Cubic, tangentMode=RCTM_User, time=0, value=0}
         {arriveTangent=7.5, interpMode=RCIM_Cubic, leaveTangent=7.5, time=0.1129, value=0.6175}
         {arriveTangent=0.443, interpMode=RCIM_Cubic, leaveTangent=0.443, time=0.2, value=1.5}
         {arriveTangent=-1.875, interpMode=RCIM_Cubic, leaveTangent=-1.875, time=0.4284, value=0.7573}
         {interpMode=RCIM_Cubic, tangentMode=RCTM_User, time=1, value=0}
        ]}}}
     Scale Curve=1.2
    )) {isSetParameters=true}
   &ParticleState()
   /SetVariables(Particles.ComponentsEnabled:NiagaraBool=&SetBoolByFloatComparison(
     A=$Particles.Age
     B=&Subtract_Float(A=$Particles.Lifetime, B=$Engine.DeltaTime)
     Comparison Type=::NewEnumerator6
    )) {isSetParameters=true}
  ]
  rendererClasses [&NiagaraComponentRendererProperties]
  renderers [{rendererClass=&NiagaraComponentRendererProperties, rendererIndex=0}]
 }
 Slash {
  emitterUpdateScript [
   &EmitterState(Life Cycle Mode=::NewEnumerator1, Loop Behavior=::NewEnumerator1, Loop Duration=1, UseLoopDelay=false)
   &SpawnBurst_Instantaneous(Use Spawn Probability=false, UseLoopCountLimit=false)
  ]
  particleSpawnScript [
   &InitializeParticle(
    Lifetime=1.9
    Mesh Scale=&Multiply_VectorByFloat(Vector=<1 1 0.3>, Float=2)
    Mesh Scale Mode=::NewEnumerator4
    UsePositionOffset=false
    Write Material Random=true
   )
   /SetVariables(Particles.MeshOrientation=$Engine.Owner.Rotation) {enabled=false, isSetParameters=true, module=SetVariables_#4}
   &CameraOffset(Camera Offset Amount=$System.CameraOffsetAll)
  ]
  particleUpdateScript [
   &ParticleState()
   &ScaleMeshSize(Scale Factor=&Multiply_Vector(
     A=$Engine.Owner.Scale
     B=&Multiply_Vector(
      A=$System.SizeScale
      B=&MakeVector(
       X=1
       Y=1
       Z=&FloatFromCurve(FloatCurve=DataInterface() {propertyValues {Curve {keys [<at 0 0.5>,<at 1 1>]}}})
      )
     )
    ))
   &UpdateMeshOrientation(
    Rotation Rate=&FloatFromCurve(
     FloatCurve=DataInterface() {propertyValues {Curve {keys [{arriveTangent=-3.0754, interpMode=RCIM_Cubic, leaveTangent=-3.0754, tangentMode=RCTM_User, time=0, value=2}, <at 0.5 0.1>]}}}
     Scale Curve=0.45
    )
    Rotation Vector=<0 0 -1>
   ) {module=UpdateMeshOrientation001}
   &ScaleColor(
    Scale RGB=&Multiply_VectorByFloat(
     Vector=&MakeVectorFromLinearColorRGB(Color=&ColorFromCurve(ColorCurve=DataInterface() {propertyValues {
         AlphaCurve {keys []}
         BlueCurve {keys [<at 0 0.5889 RCIM_Cubic>,<at 0.8853 0.5677 RCIM_Cubic>]}
         GreenCurve {keys [<at 0 0.6963 RCIM_Cubic>,<at 0.8853 0.7983 RCIM_Cubic>]}
         RedCurve {keys [<at 0 1 RCIM_Cubic>,<at 0.8853 1 RCIM_Cubic>]}
        }}))
     Float=&FloatFromCurve(FloatCurve=DataInterface() {propertyValues {Curve {keys [
          <at 0.1 1 RCIM_Cubic>
          {arriveTangent=-1.5658, interpMode=RCIM_Cubic, leaveTangent=-1.5658, tangentMode=RCTM_User, time=0.2911, value=0.9235}
          {arriveTangent=-2.0316, interpMode=RCIM_Cubic, leaveTangent=-2.0316, tangentMode=RCTM_User, time=1, value=0.1}
         ]}}})
    )
    ScaleA=true
    ScaleRGB=true
   )
   &DynamicMaterialParameters(
    Index 0 Param 1=&FloatFromCurve(FloatCurve=DataInterface() {propertyValues {Curve {keys [<at 0 0 RCIM_Cubic>,<at 1 1>]}}})
    Index 0 Param 2=0.15
    Index 0 Param 3=&FloatFromCurve(FloatCurve=DataInterface() {propertyValues {Curve {
        defaultValue=0
        keys [
         {arriveTangent=1.2925, interpMode=RCIM_Cubic, leaveTangent=1.2925, tangentMode=RCTM_User, time=0, value=0.2}
         {arriveTangent=0.6971, interpMode=RCIM_Cubic, leaveTangent=0.6971, tangentMode=RCTM_User, time=1, value=0.9}
        ]
       }}})
    Index 1 Param 1=&FloatFromCurve(FloatCurve=DataInterface() {propertyValues {Curve {keys [<at 0 0 RCIM_Cubic>,<at 1 1>]}}})
    Index 1 Param 2=&FloatFromCurve(FloatCurve=DataInterface() {propertyValues {Curve {keys [<at 0 1>,<at 0.3 0.05>,<at 1 0>]}}})
    Index 1 Param 3=&FloatFromCurve(FloatCurve=DataInterface() {propertyValues {Curve {
        defaultValue=0
        keys [
         {arriveTangent=1.2925, interpMode=RCIM_Cubic, leaveTangent=1.2925, tangentMode=RCTM_User, time=0, value=0.2}
         {arriveTangent=0.6971, interpMode=RCIM_Cubic, leaveTangent=0.6971, tangentMode=RCTM_User, time=1, value=0.9}
        ]
       }}})
    Index 2 Param 1=&FloatFromCurve(FloatCurve=DataInterface() {propertyValues {Curve {keys [<at 0 0 RCIM_Cubic>,<at 1 1>]}}})
    Index 2 Param 2=1
    Index 2 Param 3=&FloatFromCurve(FloatCurve=DataInterface() {propertyValues {Curve {
        defaultValue=0
        keys [
         {arriveTangent=1.2925, interpMode=RCIM_Cubic, leaveTangent=1.2925, tangentMode=RCTM_User, time=0, value=0.2}
         {arriveTangent=0.6971, interpMode=RCIM_Cubic, leaveTangent=0.6971, tangentMode=RCTM_User, time=1, value=0.9}
        ]
       }}})
    Module.Index0Param0Write=true
    Module.Index0Param1Write=true
    Module.Index0Param2Write=true
    Module.Index0Param3Write=true
    Module.Index1Param0Write=true
    Module.Index1Param1Write=true
    Module.Index1Param2Write=true
    Module.Index1Param3Write=true
    Module.Index2Param0Write=true
    Module.Index2Param1Write=true
    Module.Index2Param2Write=true
    Module.Index2Param3Write=true
    Write Parameter Index 1=true
    Write Parameter Index 2=true
   )
   /SetVariables() {enabled=false, isSetParameters=true, module=SetVariables_#5}
   &FN_Mesh_Attributes(
    Color=&MakeLinearColorFromVectorAndFloat(
     "Vector (RGB)"=&Multiply_Vector(
      A=&Multiply_Vector(
       A=&MakeVectorFromLinearColorRGB(Color=<rgba 0.9792 0.2808 0.1411 1>)
       B=&Multiply_VectorByFloat(
        Vector=&MakeVectorFromLinearColorRGB(Color=&ColorFromCurve(ColorCurve=DataInterface() {propertyValues {
            AlphaCurve {keys []}
            BlueCurve {keys [<at 0 0.5889 RCIM_Cubic>]}
            GreenCurve {keys [<at 0 0.6963 RCIM_Cubic>]}
            RedCurve {keys [<at 0 1 RCIM_Cubic>]}
           }}))
        Float=&FloatFromCurve(FloatCurve=DataInterface() {propertyValues {Curve {keys [
             <at 0.1 1 RCIM_Cubic>
             {arriveTangent=-1.5658, interpMode=RCIM_Cubic, leaveTangent=-1.5658, tangentMode=RCTM_User, time=0.3235, value=1}
             <at 0.6167 1.3889 RCIM_Cubic>
             {arriveTangent=-2.0316, interpMode=RCIM_Cubic, leaveTangent=-2.0316, tangentMode=RCTM_User, time=1, value=1}
            ]}}})
       )
      )
      B=&MakeVectorFromLinearColorRGB(Color=&ColorFromCurve(ColorCurve=DataInterface() {propertyValues {
          AlphaCurve {keys []}
          BlueCurve {keys [<at 0 10.9199 RCIM_Cubic>,<at 0.2578 2 RCIM_Cubic>]}
          GreenCurve {keys [<at 0 30 RCIM_Cubic>,<at 0.2578 2 RCIM_Cubic>]}
          RedCurve {keys [<at 0 25.688 RCIM_Cubic>,<at 0.2578 2 RCIM_Cubic>]}
         }}))
     )
     "Float (Alpha)"=&FloatFromCurve(
      FloatCurve=DataInterface() {propertyValues {Curve {keys [<at 0 1 RCIM_Cubic>, {arriveTangent=-3.2392, tangentMode=RCTM_User, time=1, value=0.1}]}}}
      Scale Curve=2
     )
    )
    Create Color Attribute=true
    Create Scale Attribute=true
    Scale=&Multiply_Vector(
     A=&Multiply_VectorByFloat(Vector=<1 1 0.3>, Float=2)
     B=&Multiply_Vector(
      A=$Engine.Owner.Scale
      B=&Multiply_Vector(
       A=$System.SizeScale
       B=&MakeVector(
        X=1
        Y=1
        Z=&FloatFromCurve(FloatCurve=DataInterface() {propertyValues {Curve {keys [<at 0 0.5>,<at 1 1>]}}})
       )
      )
     )
    )
    Write to Module Namespace=true
   )
   &FN_Mesh_Attributes(
    Color=&MakeLinearColorFromVectorAndFloat(
     "Vector (RGB)"=&Multiply_Vector(
      A=&Multiply_Vector(
       A=&MakeVectorFromLinearColorRGB(Color=<rgba 0.4688 0.4057 0.0905 1>)
       B=&Multiply_VectorByFloat(
        Vector=&MakeVectorFromLinearColorRGB(Color=&ColorFromCurve(ColorCurve=DataInterface() {propertyValues {
            AlphaCurve {keys []}
            BlueCurve {keys [<at 0 0.5889 RCIM_Cubic>,<at 0.8853 0.2916 RCIM_Cubic>]}
            GreenCurve {keys [<at 0 0.6963 RCIM_Cubic>,<at 0.8853 0.8929 RCIM_Cubic>]}
            RedCurve {keys [<at 0 1 RCIM_Cubic>,<at 0.8853 1 RCIM_Cubic>]}
           }}))
        Float=&FloatFromCurve(FloatCurve=DataInterface() {propertyValues {Curve {keys [
             <at 0.1 1 RCIM_Cubic>
             {arriveTangent=0.6683, interpMode=RCIM_Cubic, leaveTangent=0.6683, tangentMode=RCTM_User, time=0.2733, value=0.467}
             {arriveTangent=-0.5419, interpMode=RCIM_Cubic, leaveTangent=-0.5419, tangentMode=RCTM_User, time=1, value=1}
            ]}}})
       )
      )
      B=<1 1 1>
     )
     "Float (Alpha)"=&FloatFromCurve(FloatCurve=DataInterface() {propertyValues {Curve {keys [<at 0 1>,<at 1 0>]}}}, Scale Curve=1.8)
    )
    Create Color Attribute=true
    Create Scale Attribute=true
    Scale=&Multiply_Vector(
     A=&Multiply_VectorByFloat(Vector=<1 1 0.3>, Float=2)
     B=&Multiply_Vector(
      A=$Engine.Owner.Scale
      B=&Multiply_Vector(
       A=$System.SizeScale
       B=&MakeVector(
        X=1
        Y=1
        Z=&FloatFromCurve(FloatCurve=DataInterface() {propertyValues {Curve {keys [<at 0 0.5>,<at 1 1>]}}})
       )
      )
     )
    )
    Write to Module Namespace=true
   ) {module=FN_Mesh_Attributes001}
  ]
  rendererClasses [&NiagaraMeshRendererProperties]
  renderers [
   {rendererClass=&NiagaraMeshRendererProperties, rendererIndex=0}
   {rendererClass=&NiagaraMeshRendererProperties, rendererIndex=1}
   {rendererClass=&NiagaraMeshRendererProperties, rendererIndex=2}
  ]
 }
 SlashCenter {
  emitterUpdateScript [
   &EmitterState(Life Cycle Mode=::NewEnumerator1, Loop Behavior=::NewEnumerator1, Loop Duration=1, UseLoopDelay=false)
   &SpawnBurst_Instantaneous(Use Spawn Probability=false, UseLoopCountLimit=false)
  ]
  particleSpawnScript [
   &PartitionParticles()
   &InitializeParticle(
    Color=<rgba 20 7.3178 0 0.1>
    Lifetime=1.4
    Mesh Scale Mode=::NewEnumerator3
    Mesh Uniform Scale=2
    UsePositionOffset=true
    Write Material Random=true
   )
   /SetVariables(Particles.MeshOrientation=$Engine.Owner.Rotation) {enabled=false, isSetParameters=true}
   &CameraOffset(Camera Offset Amount=$System.CameraOffsetAll)
  ]
  particleUpdateScript [
   &ParticleState()
   &UpdateMeshOrientation(
    Rotation Rate=&FloatFromCurve(FloatCurve=DataInterface() {propertyValues {Curve {keys [{arriveTangent=-3.0754, interpMode=RCIM_Cubic, leaveTangent=-3.0754, tangentMode=RCTM_User, time=0, value=2}, <at 0.5 0.1>]}}})
    Rotation Vector=<0 0 -1>
   ) {module=UpdateMeshOrientation001}
   &ScaleMeshSize(Scale Factor=<1.3 1.3 1>) {module=ScaleMeshSize001}
   &ScaleMeshSize(Scale Factor=&Multiply_VectorByFloat(Vector=<1 1 0.7>, Float=1.05))
   &DynamicMaterialParameters(
    Index 0 Param 2=0.5
    Index 1 Param 2=0.01
    Module.Index0Float0Write=true
    Module.Index0Float1Write=true
    Module.Index0Float2Write=true
    Module.Index0Float3Write=true
    Module.Index1Float0Write=true
    Module.Index1Float1Write=true
    Module.Index1Float2Write=true
    Module.Index1Float3Write=true
    Write Parameter Index 1=::NewEnumerator1
   )
   &ScaleColor(
    Scale Alpha=&FloatFromCurve(FloatCurve=DataInterface() {propertyValues {Curve {keys [<at 0 1>,<at 0.5589 0.5647>,<at 0.8 0>]}}})
    ScaleA=true
    ScaleRGB=true
   )
   &FN_Mesh_Attributes(
    Color=&MakeLinearColorFromVectorAndFloat(
     "Vector (RGB)"=&Multiply_Vector(A=&MakeVectorFromLinearColorRGB(Color=<rgba 3 1.8281 0.7648 2>), B=<1 1 0.5>)
     "Float (Alpha)"=&Multiply_Float(
      A=2
      B=&FloatFromCurve(
       FloatCurve=DataInterface() {propertyValues {Curve {keys [
           {arriveTangent=-1.7804, time=0, value=0}
           {arriveTangent=1.4209, interpMode=RCIM_Cubic, leaveTangent=1.4209, time=0.15, value=0}
           {arriveTangent=-2.9017, interpMode=RCIM_Cubic, leaveTangent=-2.9017, tangentMode=RCTM_User, time=0.233, value=0.3311}
           {arriveTangent=-0.3887, interpMode=RCIM_Cubic, leaveTangent=-0.3887, time=0.4602, value=0.3}
           {arriveTangent=-0.5557, interpMode=RCIM_Cubic, leaveTangent=-0.5557, time=0.7435, value=0.1327}
           {arriveTangent=0.0917, tangentMode=RCTM_User, time=1, value=0}
          ]}}}
       Scale Curve=2
      )
     )
    )
    Create Color Attribute=true
    Create Scale Attribute=true
    Scale=&Multiply_Vector(A=&VectorFromFloat(Value=2.7), B=<1 1 0.4>)
    Write to Module Namespace=true
   )
  ]
  rendererClasses [&NiagaraMeshRendererProperties]
  renderers [
   {rendererClass=&NiagaraMeshRendererProperties, rendererIndex=0}
   {rendererClass=&NiagaraMeshRendererProperties, rendererIndex=1}
  ]
 }
}
summary {
 systemName=NS_SinisterMelon_SlashBigFirst
 emitters [
  {emitterName=Slash, rendererClasses [&NiagaraMeshRendererProperties]}
  {emitterName=DetalEnergy, rendererClasses [&NiagaraMeshRendererProperties]}
  {emitterName=PPV, rendererClasses [&NiagaraComponentRendererProperties]}
  {emitterName=SlashCenter, rendererClasses [&NiagaraMeshRendererProperties]}
  {emitterName=LeakParticle, simTarget=GPUComputeSim, rendererClasses [&NiagaraSpriteRendererProperties]}
 ]
 userVariables [User.bIsLocalPlayer {type=NiagaraBool}]
}
